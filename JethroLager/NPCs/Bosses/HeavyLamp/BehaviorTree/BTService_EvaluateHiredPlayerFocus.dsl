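; Behavior tree service for the HeavyLamp boss that tracks a player focus position, apparently for when the NPC has been hired.
; Each AI tick it gets the controlled pawn's Convert component and reads its converter pawn.
; When both are valid, it writes that pawn's location to the PlayerFocusKey blackboard vector.
; Invalid references leave the key untouched, and the deactivation event is empty.
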(event AI|EventReceiveTickAI
  (params (OwnerController :type AIController%20Object%20Reference) (ControlledPawn :type Pawn%20Object%20Reference) (DeltaSeconds :type Float%20%28single-precision%29))
  (bind _returnvalue (Actor|GetComponentbyClass ControlledPawn "/Script/FortniteGame.FortPawnComponent_Convert"))
  (bind _returnvalue_2 (Convert|GetConverterPawn _returnvalue))
  (Utilities|IsValid _returnvalue
    (:Is%20Valid
      (Utilities|IsValid _returnvalue_2
        (:Is%20Valid
          (AI|BehaviorTree|SetBlackboardValueasVector (Variables|Default|GetPlayerFocusKey) (Transformation|GetActorLocation _returnvalue_2)))))))

(event EventReceiveDeactivation
  (params (OwnerActor :type Actor%20Object%20Reference)))
