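# Behavior tree for NPCs in vehicles, on the Phoebe blackboard: drivers just idle here, while others can leave or fight from their seat.
# A leave-seat evaluator gates a short randomized wait followed by LeaveVehicle.
# Otherwise, weapon selection and management services run above threatened, lost-target and suspicious branches, which share vehicle ranged-attack evaluation, with an unaware scanning fallback.
# Evaluator services do the work, as alert branches hold only Wait tasks.
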BehaviorTree /Game/Athena/AI/NPCs/Base/BehaviorTree/BT_NPC_InVehicle
blackboard=&BB_Phoebe
root {
 name=Selector
 class=BTComposite_Selector
 children [
  Selector {
   class=BTComposite_Selector
   children [Wait {class=BTTask_Wait, properties {waitTime {defaultValue=5}}}]
   decorators [Is Driving Vehicle {
     class=BTDecorator_Blackboard
     properties {
      arithmeticOperation=Equal
      intValue=0
      nodeName=Is Driving Vehicle
      notifyObserver=ValueChange
      stringValue=""
      blackboardKey {selectedKeyName=AIEvaluator_IsDrivingVehicle}
     }
    }]
   properties {
    children [{childComposite=null, childTask=&self:BTTask_Wait_0, decorators []}]
    services []
   }
   services []
  }
  Selector {
   class=BTComposite_Selector
   children [
    Sequence {
     class=BTComposite_Sequence
     children [
      Wait {class=BTTask_Wait, properties {randomDeviation {defaultValue=0.5}, waitTime {defaultValue=0.6}}}
      LeaveVehicle {
       class=FortAthenaBTTask_LeaveVehicle
       children []
       decorators []
       properties {bIgnoreRestartSelf=false, bWaitVehicleStop=false, nodeName=LeaveVehicle}
       services []
      }
     ]
     decorators [Should leave vehicle {
       class=BTDecorator_Blackboard
       properties {
        arithmeticOperation=GreaterOrEqual
        intValue=4
        nodeName=Should leave vehicle
        stringValue=EExecutionStatus::ExecutionAllowed
        blackboardKey {selectedKeyName=AIEvaluator_LeaveSeatStatus}
       }
      }]
     properties {children [
       {childTask=&self:BTTask_Wait_2}
       {childTask=&self:FortAthenaBTTask_LeaveVehicle_0}
      ]}
    }
    Selector {
     class=BTComposite_Selector
     children [
      Selector {
       class=BTComposite_Selector
       children [
        THREATENED {
         class=BTComposite_Selector
         children [Wait {class=BTTask_Wait, properties {waitTime {defaultValue=5}}}]
         decorators [IsThreatened {
           class=BTDecorator_Blackboard
           properties {
            arithmeticOperation=Equal
            intValue=3
            nodeName=IsThreatened
            stringValue=EAlertLevel::Threatened
            blackboardKey {selectedKeyName=AIEvaluator_Global_AlertLevel}
           }
          }]
         properties {
          nodeName=THREATENED
          children [{childComposite=null, childTask=&self:BTTask_Wait_6, decorators []}]
          services [&self:FortAthenaBTService_AIEvaluator_40]
         }
         services [Focus on target {
           class=FortAthenaBTService_AIEvaluator
           properties {interval=0.5, nodeName=Focus on target, aIEvaluatorClass {defaultValue=&FortAthenaAIBotEvaluator_HandleFocusing}}
          }]
        }
        LOST TARGET {
         class=BTComposite_Selector
         children [Wait {class=BTTask_Wait, properties {waitTime {defaultValue=5}}}]
         decorators [Has Lost Target {
           class=BTDecorator_Blackboard
           properties {
            arithmeticOperation=Equal
            intValue=2
            nodeName=Has Lost Target
            stringValue=EAlertLevel::LKP
            blackboardKey {selectedKeyName=AIEvaluator_Global_AlertLevel}
           }
          }]
         properties {
          nodeName=LOST TARGET
          children [{childComposite=null, childTask=&self:BTTask_Wait_5, decorators []}]
          services [&self:FortAthenaBTService_AIEvaluator_35]
         }
         services [Focus Scan Around or Heard Sound {
           class=FortAthenaBTService_AIEvaluator
           properties {
            interval=0.5
            nodeName=Focus Scan Around or Heard Sound
            aIEvaluatorClass {defaultValue=/Game/Athena/AI/Phoebe/Evaluators/Phoebe_HandleFocusing_FocusInvestigation.Phoebe_HandleFocusing_FocusInvestigation_C}
           }
          }]
        }
        SUSPICIOUS {
         class=BTComposite_Selector
         children [Wait {class=BTTask_Wait, properties {waitTime {defaultValue=5}}}]
         decorators [Is Suspicious {
           class=BTDecorator_Blackboard
           properties {
            arithmeticOperation=Equal
            intValue=1
            nodeName=Is Suspicious
            stringValue=EAlertLevel::Alerted
            blackboardKey {selectedKeyName=AIEvaluator_Global_AlertLevel}
           }
          }]
         properties {
          nodeName=SUSPICIOUS
          children [{childComposite=null, childTask=&self:BTTask_Wait_3, decorators []}]
          services [
           &self:FortAthenaBTService_AIEvaluator_37
           &self:FortAthenaBTService_AIEvaluator_38
          ]
         }
         services [
          Evaluating...Observe Threat {
           class=FortAthenaBTService_AIEvaluator
           properties {
            bCallTickOnSearchStart=true
            interval=0.5
            nodeName=Evaluating...Observe Threat
            aIEvaluatorClass {defaultValue=/Script/FortniteAIServer.FortAthenaAIBotEvaluator_Observe}
           }
          }
          Focus on target {
           class=FortAthenaBTService_AIEvaluator
           properties {interval=0.5, nodeName=Focus on target, aIEvaluatorClass {defaultValue=&FortAthenaAIBotEvaluator_HandleFocusing}}
          }
         ]
        }
       ]
       decorators []
       properties {
        children [
         {
          childComposite=&self:BTComposite_Selector_24
          childTask=null
          decorators [&self:BTDecorator_Blackboard_10]
         }
         {
          childComposite=&self:BTComposite_Selector_22
          childTask=null
          decorators [&self:BTDecorator_Blackboard_8]
         }
         {
          childComposite=&self:BTComposite_Selector_23
          childTask=null
          decorators [&self:BTDecorator_Blackboard_9]
         }
        ]
        services [&self:FortAthenaBTService_AIEvaluator_30]
       }
       services [Evaluating...Vehicle Range Attack {
         class=FortAthenaBTService_AIEvaluator
         properties {
          bCallTickOnSearchStart=true
          interval=0.2
          nodeName=Evaluating...Vehicle Range Attack
          randomDeviation=0.05
          aIEvaluatorClass {defaultValue=/Game/Athena/AI/Vehicle/Evaluators/VehicleRangeAttackEvaluator.VehicleRangeAttackEvaluator_C}
         }
        }]
      }
      UNAWARE {
       class=BTComposite_Selector
       children [Wait {class=BTTask_Wait, properties {waitTime {defaultValue=5}}}]
       decorators []
       properties {
        nodeName=UNAWARE
        children [{childComposite=null, childTask=&self:BTTask_Wait_4, decorators []}]
        services [
         &self:FortAthenaBTService_AIEvaluator_2
         &self:FortAthenaBTService_AIEvaluator_4
        ]
       }
       services [
        Evaluating...Range Attack {
         class=FortAthenaBTService_AIEvaluator
         properties {
          interval=1
          nodeName=Evaluating...Range Attack
          aIEvaluatorClass {defaultValue=&FortAthenaAIBotEvaluator_RangeAttack}
         }
        }
        Focus Scan Around {
         class=FortAthenaBTService_AIEvaluator
         properties {
          interval=1
          nodeName=Focus Scan Around
          randomDeviation=0.5
          aIEvaluatorClass {defaultValue=/Game/Athena/AI/Phoebe/Evaluators/Phoebe_HandleFocusing_ScanAround.Phoebe_HandleFocusing_ScanAround_C}
         }
        }
       ]
      }
     ]
     decorators []
     properties {
      children [
       {childComposite=&self:BTComposite_Selector_19, childTask=null, decorators []}
       {childComposite=&self:BTComposite_Selector_10, childTask=null, decorators []}
      ]
      services [
       &self:FortAthenaBTService_AIEvaluator_34
       &self:FortAthenaBTService_ManageWeapon_3
      ]
     }
     services [
      Evaluating...Select Weapon {
       class=FortAthenaBTService_AIEvaluator
       properties {
        interval=0.2
        nodeName=Evaluating...Select Weapon
        randomDeviation=0.05
        aIEvaluatorClass {defaultValue=&FortAthenaAIBotEvaluator_WeaponSelection}
       }
      }
      Executing...Weapons Actions {class=FortAthenaBTService_ManageWeapon, properties {interval=0.2, randomDeviation=0.1}}
     ]
    }
   ]
   decorators [Loop {class=BTDecorator_Loop, properties {numLoops {defaultValue=1}}}]
   properties {
    children [
     {
      childComposite=&self:BTComposite_Sequence_0
      childTask=null
      decorators [&self:BTDecorator_Blackboard_0]
     }
     {childComposite=&self:BTComposite_Selector_21, childTask=null, decorators []}
    ]
    services [&self:FortAthenaBTService_AIEvaluator_0]
   }
   services [Evaluating...Leave Vehicle {
     class=FortAthenaBTService_AIEvaluator
     properties {
      interval=0.5
      nodeName=Evaluating...Leave Vehicle
      aIEvaluatorClass {defaultValue=/Game/Athena/AI/NPCs/Base/Evaluators/Evaluator_LeaveSeatVehicle_NPC.Evaluator_LeaveSeatVehicle_NPC_C}
     }
    }]
  }
 ]
 decorators []
 properties {
  children [
   {
    childComposite=&self:BTComposite_Selector_1
    childTask=null
    decorators [&self:BTDecorator_Blackboard_15]
   }
   {
    childComposite=&self:BTComposite_Selector_4
    childTask=null
    decorators [&self:BTDecorator_Loop_2]
   }
  ]
  services []
 }
 services []
}
