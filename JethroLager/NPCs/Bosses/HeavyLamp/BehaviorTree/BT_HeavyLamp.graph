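# Behavior tree for the HeavyLamp boss NPC, a priority selector from diving, gliding, vehicles and injected behavior down to on-ground logic.
# Ground branches handle launches, traps, threat avoidance, DBNO and leash return, with teleport or catch-up leap when converted.
# Threatened state tries force throw, pulls, leaps and lightsaber throw before looping attack movement; non-attacking branches include last-known-position investigation.
# AI evaluator services feed blackboard keys that gate branches alongside behavior-control and ability checks.
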BehaviorTree /JethroLager/NPCs/Bosses/HeavyLamp/BehaviorTree/BT_HeavyLamp
blackboard=&/BB_Vk_HeavyLamp
root {
 name=Alive
 class=BTComposite_Selector
 children [
  Diving {
   class=FortAthenaBTTask_Dive
   children []
   decorators [
    Can Dive {
     class=BTDecorator_Blackboard
     properties {
      arithmeticOperation=Equal
      intValue=4
      nodeName=Can Dive
      stringValue=EExecutionStatus::ExecutionAllowed
      blackboardKey {selectedKeyName=AIEvaluator_Dive_ExecutionStatus}
     }
    }
    "Behavior Controls = CanUseDiving" {
     class=FortAthenaBTDecorator_BehaviorControls
     properties {bInverseCondition=false, behaviorTreeBranch=CanUseDiving, flowAbortMode=Both, nodeName="Behavior Controls = CanUseDiving"}
    }
   ]
   properties {
    bIgnoreRestartSelf=false
    diveDestinationKeyName=AIEvaluator_Dive_Destination
    executionStatusKeyName=AIEvaluator_Dive_ExecutionStatus
    nodeName=Diving
   }
   services []
  }
  Gliding {
   class=FortAthenaBTTask_Glide
   children []
   decorators [
    Can Glide {
     class=BTDecorator_Blackboard
     properties {
      arithmeticOperation=Equal
      intValue=4
      nodeName=Can Glide
      stringValue=EExecutionStatus::ExecutionAllowed
      blackboardKey {selectedKeyName=AIEvaluator_Glide_ExecutionStatus}
     }
    }
    "Behavior Controls = CanUseGliding" {
     class=FortAthenaBTDecorator_BehaviorControls
     properties {bInverseCondition=false, behaviorTreeBranch=CanUseGliding, flowAbortMode=Both, nodeName="Behavior Controls = CanUseGliding"}
    }
   ]
   properties {
    bIgnoreRestartSelf=false
    executionStatusKeyName=AIEvaluator_Glide_ExecutionStatus
    glideDestinationKeyName=AIEvaluator_Glide_Destination
    nodeName=Gliding
   }
   services []
  }
  Vehicle {
   class=BTTask_RunBehavior
   decorators [
    Is in a Vehicle {
     class=BTDecorator_Blackboard
     properties {
      arithmeticOperation=Equal
      intValue=0
      nodeName=Is in a Vehicle
      notifyObserver=ValueChange
      stringValue=""
      blackboardKey {selectedKeyName=AIEvaluator_Vehicle_Actor}
     }
    }
    Loop {class=BTDecorator_Loop}
   ]
   properties {behaviorAsset=/Game/Athena/AI/NPCs/Base/BehaviorTree/BT_NPC_InVehicle, nodeName=Vehicle}
  }
  Injection Point Alive.HighPriority {
   class=FortBTTask_InjectionPoint
   children []
   decorators [Can Do Injected Behavior {
     class=FortBTDecorator_CanDoInjectedBehavior
     properties {
      arithmeticOperation=GreaterOrEqual
      bInverseCondition=false
      basicOperation=Set
      floatValue=0
      flowAbortMode=Both
      intValue=3
      nodeName=Can Do Injected Behavior
      notifyObserver=ResultChange
      stringValue=ExecutionPending
      textOperation=Equal
      blackboardKey {bNoneIsAllowedValue=false, selectedKeyName=AIEvaluator_Injection_Alive_HighPriority}
     }
    }]
   properties {
    bIgnoreRestartSelf=false
    nodeName=Injection Point Alive.HighPriority
    injectedAction {baseClass=&Object, defaultValue=null, key=Injection_Alive_HighPriority_ActionObject}
    injectionTag {tagName=Behavior.Injection.Alive.HighPriority}
   }
   services []
  }
  On Ground {
   class=BTComposite_Selector
   children [
    Wait {
     class=BTTask_Wait
     decorators [
      Is EndGame? {
       class=BTDecorator_Blackboard
       properties {
        arithmeticOperation=GreaterOrEqual
        intValue=11
        nodeName=Is EndGame?
        stringValue=EAthenaGamePhaseStep::EndGame
        blackboardKey {selectedKeyName=AIEvaluator_Global_GamePhaseStep}
       }
      }
      Loop {class=BTDecorator_Loop}
     ]
     properties {waitTime {defaultValue=5}}
    }
    Wait {
     class=BTTask_Wait
     decorators [
      Is Movement Blocked? {
       class=BTDecorator_Blackboard
       properties {
        arithmeticOperation=Equal
        intValue=0
        nodeName=Is Movement Blocked?
        stringValue=""
        blackboardKey {selectedKeyName=AIEvaluator_Global_IsMovementBlocked}
       }
      }
      Loop {class=BTDecorator_Loop}
     ]
     properties {waitTime {defaultValue=1}}
    }
    Steer Movement {
     class=FortAthenaBTTask_SteerMovement
     children []
     decorators [
      Is Character being launched? {
       class=BTDecorator_Blackboard
       properties {
        arithmeticOperation=GreaterOrEqual
        intValue=3
        nodeName=Is Character being launched?
        stringValue=EExecutionStatus::ExecutionPending
        blackboardKey {selectedKeyName=AIEvaluator_CharacterLaunched_ExecutionStatus}
       }
      }
      Has Active Ability with Any Matching Tag {
       class=HasActiveAbilityWithAnyMatchingTag_C
       properties {
        bCheckConditionOnlyBlackBoardChanges=false
        bInverseCondition=true
        bShowPropertyDetails=true
        customDescription=""
        flowAbortMode=None
        nodeName=Has Active Ability with Any Matching Tag
        activeAbilityTagsToCheck {gameplayTags [{tagName=Jethro.HeavyLamp.Abilities.CatchUpLeap}]}
       }
      }
     ]
     properties {
      bIgnoreRestartSelf=false
      bSetControlRotation=false
      nodeName=Steer Movement
      steerDirectionKeySelector {bNoneIsAllowedValue=false, selectedKeyName=AIEvaluator_CharacterLaunched_SteerDirection}
     }
     services []
    }
    Bot Shoots Trap {
     class=FortAthenaBTTask_ShootTrap
     children []
     decorators [
      Trap on Path? {
       class=BTDecorator_Blackboard
       properties {
        arithmeticOperation=GreaterOrEqual
        intValue=3
        nodeName=Trap on Path?
        stringValue=EExecutionStatus::ExecutionPending
        blackboardKey {selectedKeyName=AIEvaluator_TrapOnPath_ExecutionStatus}
       }
      }
      "Behavior Controls = CanReactToTraps" {
       class=FortAthenaBTDecorator_BehaviorControls
       properties {bInverseCondition=false, behaviorTreeBranch=CanReactToTraps, flowAbortMode=Both, nodeName="Behavior Controls = CanReactToTraps"}
      }
     ]
     properties {
      bIgnoreRestartSelf=false
      nodeName=Bot Shoots Trap
      trapOnPathKeyName=AIEvaluator_TrapOnPath_ExecutionStatus
      randomDeviation {defaultValue=0, key=None}
      targetActorKey {bNoneIsAllowedValue=false, selectedKeyName=AIEvaluator_TrapOnPath_Actor}
      waitTime {defaultValue=5, key=None}
     }
     services []
    }
    AvoidThreat MoveTo {
     class=FortAthenaBTTask_BotMoveTo
     decorators [
      Can AvoidThreat? {
       class=BTDecorator_Blackboard
       properties {
        arithmeticOperation=GreaterOrEqual
        intValue=4
        nodeName=Can AvoidThreat?
        stringValue=EExecutionStatus::ExecutionAllowed
        blackboardKey {selectedKeyName=AIEvaluator_AvoidThreat_ExecutionStatus}
       }
      }
      "Behavior Controls = CanReactToProjectile" {
       class=FortAthenaBTDecorator_BehaviorControls
       properties {bInverseCondition=false, behaviorTreeBranch=CanReactToProjectile, flowAbortMode=Both, nodeName="Behavior Controls = CanReactToProjectile"}
      }
     ]
     properties {
      movementResultKeyName=AIEvaluator_AvoidThreat_MovementState
      nodeName=AvoidThreat MoveTo
      acceptableRadius {defaultValue=100}
      allowRandomWobble {defaultValue=true}
      bAllowStrafe {defaultValue=true}
      blackboardKey {selectedKeyName=AIEvaluator_AvoidThreat_Destination}
      observedBlackboardValueTolerance {defaultValue=4.75}
     }
    }
    Selector {
     class=BTComposite_Selector
     children [
      Bot Move To {
       class=FortAthenaBTTask_BotMoveTo
       decorators [Can Do DBNO Move {
         class=BTDecorator_Blackboard
         properties {
          arithmeticOperation=GreaterOrEqual
          intValue=3
          nodeName=Can Do DBNO Move
          stringValue=EExecutionStatus::ExecutionPending
          blackboardKey {selectedKeyName=AIEvaluator_DBNO_ExecutionStatus}
         }
        }]
       properties {
        movementResultKeyName=AIEvaluator_DBNO_MovementState
        nodeName=Bot Move To
        acceptableRadius {defaultValue=150}
        allowRandomWobble {defaultValue=true}
        bAllowStrafe {defaultValue=false}
        blackboardKey {selectedKeyName=AIEvaluator_DBNO_Destination}
        filterClass {defaultValue=&NavFilter_Phoebe_NoSmashable_C}
        observedBlackboardValueTolerance {defaultValue=4.75}
       }
      }
      Wait {class=BTTask_Wait, properties {waitTime {defaultValue=1}}}
     ]
     decorators [
      Is DBNO {
       class=BTDecorator_Blackboard
       properties {arithmeticOperation=GreaterOrEqual, intValue=3, nodeName=Is DBNO, stringValue="", blackboardKey {selectedKeyName=AIEvaluator_Global_DBNO}}
      }
      Loop {class=BTDecorator_Loop}
     ]
     properties {
      children [
       {
        childComposite=null
        childTask=&self:FortAthenaBTTask_BotMoveTo_2
        decorators [&self:BTDecorator_Blackboard_10]
       }
       {childComposite=null, childTask=&self:BTTask_Wait_4, decorators []}
      ]
      services [&self:FortAthenaBTService_AIEvaluator_22]
     }
     services [Evaluating...DBNO Move {
       class=FortAthenaBTService_AIEvaluator
       properties {
        bCallTickOnSearchStart=true
        interval=0.1
        nodeName=Evaluating...DBNO Move
        randomDeviation=0
        aIEvaluatorClass {defaultValue=/Game/Athena/AI/NPCs/Base/Evaluators/Evaluator_DBNO_Guard.Evaluator_DBNO_Guard_C}
       }
      }]
    }
    Sequence {
     class=BTComposite_Sequence
     children [
      MoveTo Vehicle {
       class=FortAthenaBTTask_BotMoveTo
       decorators []
       properties {
        movementResultKeyName=None
        nodeName=MoveTo Vehicle
        acceptableRadius {defaultValue=250}
        allowRandomWobble {defaultValue=true}
        bAllowStrafe {defaultValue=false}
        blackboardKey {selectedKeyName=AIEvaluator_SelectedVehicle}
        observedBlackboardValueTolerance {defaultValue=3}
       }
      }
      FortAthenaBTTask_EnterVehicle_0 {
       class=FortAthenaBTTask_EnterVehicle
       children []
       decorators []
       properties {bIgnoreRestartSelf=false, nodeName="", selectedVehicleKeyName=AIEvaluator_SelectedVehicle}
       services []
      }
     ]
     decorators [
      Is Vehicle Selected? {
       class=BTDecorator_Blackboard
       properties {
        arithmeticOperation=GreaterOrEqual
        intValue=3
        nodeName=Is Vehicle Selected?
        stringValue=ExecutionPending
        blackboardKey {selectedKeyName=AIEvaluator_SelectVehicle_Status}
       }
      }
      Vehicle Object Valid {
       class=BTDecorator_Blackboard
       properties {
        arithmeticOperation=GreaterOrEqual
        intValue=4
        nodeName=Vehicle Object Valid
        stringValue=""
        blackboardKey {selectedKeyName=AIEvaluator_SelectedVehicle}
       }
      }
     ]
     properties {
      children [
       {childTask=&self:FortAthenaBTTask_BotMoveTo_5}
       {childTask=&self:FortAthenaBTTask_EnterVehicle_0}
      ]
      services [
       &self:FortAthenaBTService_AIEvaluator_13
       &self:FortAthenaBTService_AIEvaluator_23
       &self:FortAthenaBTService_ManageWeapon_1
      ]
     }
     services [
      Evaluating...Select Weapon {
       class=FortAthenaBTService_AIEvaluator
       properties {
        interval=0.5
        nodeName=Evaluating...Select Weapon
        randomDeviation=0.05
        aIEvaluatorClass {defaultValue=&FortAthenaAIBotEvaluator_WeaponSelection}
       }
      }
      Evaluating...Melee Attack {
       class=FortAthenaBTService_AIEvaluator
       properties {
        interval=0.3
        nodeName=Evaluating...Melee Attack
        aIEvaluatorClass {defaultValue=&FortAthenaAIBotEvaluator_MeleeAttack}
       }
      }
      Executing...Weapons Actions {class=FortAthenaBTService_ManageWeapon, properties {interval=1, randomDeviation=0.3}}
     ]
    }
    Selector {
     class=BTComposite_Selector
     children [
      Selector {
       class=BTComposite_Selector
       children [
        GameplayAbility {
         class=FortBTTask_ExecuteGameplayAbility
         decorators [
          Should Teleport Towards Converter? {
           class=BTDecorator_Blackboard
           properties {
            arithmeticOperation=Equal
            intValue=0
            nodeName=Should Teleport Towards Converter?
            stringValue=""
            blackboardKey {selectedKeyName=AIEvaluator_Converted_ShouldTeleportTowardsConverter}
           }
          }
          Has Converted Teleport Ability? {
           class=FortBTDecorator_GameplayAbility_HasGameplayAbility
           properties {nodeName=Has Converted Teleport Ability?, gameplayAbilityTagContainer {defaultValue {gameplayTags [{tagName=Ability.Converted.Teleport}]}}}
          }
          Can Activate Converted Teleport Ability? {
           class=FortBTDecorator_GameplayAbility_CanActivate
           properties {
            flowAbortMode=Both
            nodeName=Can Activate Converted Teleport Ability?
            gameplayAbilityTagContainer {defaultValue {gameplayTags [{tagName=Ability.Converted.Teleport}]}}
            targetValue {key=SelfActor}
           }
          }
         ]
         properties {nodeName=GameplayAbility, gameplayAbilityTagContainer {defaultValue {gameplayTags [{tagName=Ability.Converted.Teleport}]}}}
        }
        Sequence {
         class=BTComposite_Sequence
         children [
          Rotate to face BB entry {
           class=BTTask_RotateToFaceBBEntry
           decorators []
           properties {blackboardKey {selectedKeyName=AIEvaluator_Converted_Destination}, precision {defaultValue=2}}
          }
          Leap {
           class=FortBTTask_ExecuteGameplayAbility
           decorators []
           properties {nodeName=Leap, gameplayAbilityTagContainer {defaultValue {gameplayTags [{tagName=Jethro.HeavyLamp.Abilities.CatchUpLeap}]}}}
          }
         ]
         decorators [
          Has Gameplay Ability {
           class=FortBTDecorator_GameplayAbility_HasGameplayAbility
           properties {gameplayAbilityTagContainer {defaultValue {gameplayTags [{tagName=Jethro.HeavyLamp.Abilities.CatchUpLeap}]}}}
          }
          Can Activate Gameplay Ability {
           class=FortBTDecorator_GameplayAbility_CanActivate
           properties {
            flowAbortMode=LowerPriority
            gameplayAbilityTagContainer {defaultValue {gameplayTags [{tagName=Jethro.HeavyLamp.Abilities.CatchUpLeap}]}}
            targetValue {key=SelfActor}
           }
          }
         ]
         properties {children [
           {childTask=&self:BTTask_RotateToFaceBBEntry_7}
           {childTask=&self:FortBTTask_ExecuteGameplayAbility_7}
          ]}
        }
       ]
       decorators [Is Converted? {
         class=BTDecorator_GameplayTagQuery
         properties {
          flowAbortMode=Both
          nodeName=Is Converted?
          gameplayTagQuery {
           autoDescription=" ANY( Athena.Gameplay.Status.Converted )"
           queryTokenStream [0,1,1,1,0]
           tagDictionary [{tagName=Athena.Gameplay.Status.Converted}]
          }
         }
        }]
       properties {
        children [
         {
          childComposite=null
          childTask=&self:FortBTTask_ExecuteGameplayAbility_6
          decorators [
           &self:BTDecorator_Blackboard_43
           &self:FortBTDecorator_GameplayAbility_HasGameplayAbility_13
           &self:FortBTDecorator_GameplayAbility_CanActivate_16
          ]
         }
         {
          childComposite=&self:BTComposite_Sequence_9
          childTask=null
          decorators [
           &self:FortBTDecorator_GameplayAbility_HasGameplayAbility_10
           &self:FortBTDecorator_GameplayAbility_CanActivate_10
          ]
         }
        ]
        services []
       }
       services []
      }
      Run Unstuck Behavior {
       class=BTTask_RunBehavior
       decorators []
       properties {behaviorAsset=&BT_Phoebe_UnstuckNew, nodeName=Run Unstuck Behavior}
      }
      Selector {
       class=BTComposite_Selector
       children [
        Run Zipline Behavior {
         class=BTTask_RunBehavior
         decorators [
          Blackboard Based Condition {
           class=BTDecorator_Blackboard
           properties {
            arithmeticOperation=GreaterOrEqual
            intValue=4
            nodeName=Blackboard Based Condition
            stringValue=EExecutionStatus::ExecutionAllowed
            blackboardKey {selectedKeyName=AIEvaluator_Zipline_MoveExecutionStatus}
           }
          }
          Blackboard Based Condition {
           class=BTDecorator_Blackboard
           properties {
            arithmeticOperation=GreaterOrEqual
            intValue=4
            nodeName=Blackboard Based Condition
            stringValue=EExecutionStatus::ExecutionAllowed
            blackboardKey {selectedKeyName=AIEvaluator_Zipline_UsageExecutionStatus}
           }
          }
         ]
         properties {behaviorAsset=&BT_Phoebe_Zipline, nodeName=Run Zipline Behavior}
        }
        Back into Leash MoveTo {
         class=FortAthenaBTTask_BotMoveTo
         decorators []
         properties {
          movementResultKeyName=AIEvaluator_Leash_MovementState
          nodeName=Back into Leash MoveTo
          acceptableRadius {defaultValue=10}
          allowRandomWobble {defaultValue=true}
          bAllowStrafe {defaultValue=true}
          blackboardKey {selectedKeyName=AIEvaluator_Leash_Destination}
          observedBlackboardValueTolerance {defaultValue=4.75}
         }
        }
       ]
       decorators []
       properties {
        children [
         {
          childComposite=null
          childTask=&self:BTTask_RunBehavior_23
          decoratorOps [{number=2, operation=Or}, {number=0}, {number=1}]
          decorators [
           &self:BTDecorator_Blackboard_25
           &self:BTDecorator_Blackboard_28
          ]
         }
         {childComposite=null, childTask=&self:FortAthenaBTTask_BotMoveTo_4, decorators []}
        ]
        services [
         &self:FortAthenaBTService_AIEvaluator_29
         &self:FortAthenaBTService_AIEvaluator_80
        ]
       }
       services [
        Evaluating... Zipline {
         class=FortAthenaBTService_AIEvaluator
         properties {
          bCallTickOnSearchStart=true
          interval=0.5
          nodeName=Evaluating... Zipline
          aIEvaluatorClass {defaultValue=/Game/Athena/AI/NPCs/Base/Evaluators/Evaluator_Zipline_BackToLeash.Evaluator_Zipline_BackToLeash_C}
         }
        }
        Evaluate...Sprinting {
         class=FortAthenaBTService_AIEvaluator
         properties {interval=0.2, nodeName=Evaluate...Sprinting, aIEvaluatorClass {defaultValue=/Script/FortniteAIServer.FortAthenaAIBotEvaluator_Sprinting}}
        }
       ]
      }
     ]
     decorators [
      Must Move back into Leash? {
       class=BTDecorator_Blackboard
       properties {
        arithmeticOperation=GreaterOrEqual
        intValue=4
        nodeName=Must Move back into Leash?
        stringValue=EExecutionStatus::ExecutionAllowed
        blackboardKey {selectedKeyName=AIEvaluator_Leash_ExecutionStatus}
       }
      }
      "Behavior Controls = CanReturnToLeash" {
       class=FortAthenaBTDecorator_BehaviorControls
       properties {bInverseCondition=false, behaviorTreeBranch=CanReturnToLeash, flowAbortMode=Both, nodeName="Behavior Controls = CanReturnToLeash"}
      }
      Is Converted or Turned against converter? {
       class=BTDecorator_GameplayTagQuery
       properties {
        flowAbortMode=Both
        nodeName=Is Converted or Turned against converter?
        gameplayTagQuery {
         autoDescription=" ANY( Athena.Gameplay.Status.Converted, Athena.Gameplay.Status.TurnedAgainstConverter )"
         queryTokenStream [0,1,1,2,0,1]
         tagDictionary [{tagName=Athena.Gameplay.Status.Converted}, {tagName=Athena.Gameplay.Status.TurnedAgainstConverter}]
        }
       }
      }
     ]
     properties {
      children [
       {
        childComposite=&self:BTComposite_Selector_11
        childTask=null
        decorators [&self:BTDecorator_GameplayTagQuery_4]
       }
       {childComposite=null, childTask=&self:BTTask_RunBehavior_1, decorators []}
       {childComposite=&self:BTComposite_Selector_22, childTask=null, decorators []}
      ]
      services [
       &self:FortAthenaBTService_AIEvaluator_78
       &self:FortAthenaBTService_AIEvaluator_41
       &self:FortAthenaBTService_AIEvaluator_40
       &self:FortAthenaBTService_AIEvaluator_42
       &self:FortAthenaBTService_ManageWeapon_2
       &self:FortAthenaBTService_AIEvaluator_21
      ]
     }
     services [
      Evaluating...Select Weapon {
       class=FortAthenaBTService_AIEvaluator
       properties {
        interval=0.5
        nodeName=Evaluating...Select Weapon
        randomDeviation=0.05
        aIEvaluatorClass {defaultValue=&FortAthenaAIBotEvaluator_WeaponSelection}
       }
      }
      Evaluating...Range Attack {
       class=FortAthenaBTService_AIEvaluator
       properties {
        bCallTickOnSearchStart=true
        interval=0.2
        nodeName=Evaluating...Range Attack
        randomDeviation=0.05
        aIEvaluatorClass {defaultValue=&FortAthenaAIBotEvaluator_RangeAttack}
       }
      }
      Evaluating...Melee Attack {
       class=FortAthenaBTService_AIEvaluator
       properties {
        interval=0.3
        nodeName=Evaluating...Melee Attack
        aIEvaluatorClass {defaultValue=&FortAthenaAIBotEvaluator_MeleeAttack}
       }
      }
      Evaluating... Evasive Maneuvers {
       class=FortAthenaBTService_AIEvaluator
       properties {
        interval=0.5
        nodeName=Evaluating... Evasive Maneuvers
        aIEvaluatorClass {defaultValue=/Game/Athena/AI/Phoebe/Evaluators/Phoebe_EvasiveManeuvers.Phoebe_EvasiveManeuvers_C}
       }
      }
      Executing...Weapons Actions {class=FortAthenaBTService_ManageWeapon, properties {interval=1, randomDeviation=0.3}}
      Evaluating... Converted Actions {
       class=FortAthenaBTService_AIEvaluator
       properties {
        interval=0.5
        nodeName=Evaluating... Converted Actions
        aIEvaluatorClass {defaultValue=/Script/FortniteAIServer.FortAthenaAIBotEvaluator_Converted}
       }
      }
     ]
    }
    THREATEN {
     class=BTComposite_Selector
     children [
      Sequence {
       class=BTComposite_Sequence
       children [
        Rotate to face BB entry {
         class=BTTask_RotateToFaceBBEntry
         decorators []
         properties {blackboardKey {selectedKeyName=AIEvaluator_Global_TargetActor}, precision {defaultValue=2}}
        }
        Force Throw {
         class=FortBTTask_ExecuteGameplayAbility
         decorators []
         properties {nodeName=Force Throw, gameplayAbilityTagContainer {defaultValue {gameplayTags [{tagName=NPCLibrary.Abilities.TKThrow}]}}}
        }
       ]
       decorators [
        Has Gameplay Ability {
         class=FortBTDecorator_GameplayAbility_HasGameplayAbility
         properties {gameplayAbilityTagContainer {defaultValue {gameplayTags [{tagName=NPCLibrary.Abilities.TKThrow}]}}}
        }
        Can Activate Gameplay Ability {
         class=FortBTDecorator_GameplayAbility_CanActivate
         properties {
          flowAbortMode=Both
          gameplayAbilityTagContainer {defaultValue {gameplayTags [{tagName=NPCLibrary.Abilities.TKThrow}]}}
          targetValue {key=SelfActor}
         }
        }
       ]
       properties {children [
         {childTask=&self:BTTask_RotateToFaceBBEntry_0}
         {childTask=&self:FortBTTask_ExecuteGameplayAbility_1}
        ]}
      }
      Sequence {
       class=BTComposite_Sequence
       children [
        Rotate to face BB entry {
         class=BTTask_RotateToFaceBBEntry
         decorators []
         properties {blackboardKey {selectedKeyName=AIEvaluator_Global_TargetActor}, precision {defaultValue=2}}
        }
        Force Pull {
         class=FortBTTask_ExecuteGameplayAbility
         decorators []
         properties {nodeName=Force Pull, gameplayAbilityTagContainer {defaultValue {gameplayTags [{tagName=Ability.Realm.ForcePull}]}}}
        }
       ]
       decorators [
        Has Gameplay Ability {
         class=FortBTDecorator_GameplayAbility_HasGameplayAbility
         properties {gameplayAbilityTagContainer {defaultValue {gameplayTags [{tagName=Ability.Realm.ForcePull}]}}}
        }
        Can Activate Gameplay Ability {
         class=FortBTDecorator_GameplayAbility_CanActivate
         properties {
          flowAbortMode=Both
          gameplayAbilityTagContainer {defaultValue {gameplayTags [{tagName=Ability.Realm.ForcePull}]}}
          targetValue {key=SelfActor}
         }
        }
       ]
       properties {children [
         {childTask=&self:BTTask_RotateToFaceBBEntry_2}
         {childTask=&self:FortBTTask_ExecuteGameplayAbility_3}
        ]}
      }
      Sequence {
       class=BTComposite_Sequence
       children [
        Rotate to face BB entry {
         class=BTTask_RotateToFaceBBEntry
         decorators []
         properties {blackboardKey {selectedKeyName=AIEvaluator_Global_TargetActor}, precision {defaultValue=2}}
        }
        Force Pull {
         class=FortBTTask_ExecuteGameplayAbility
         decorators []
         properties {nodeName=Force Pull, gameplayAbilityTagContainer {defaultValue {gameplayTags [{tagName=Ability.Realm.ForcePullRebootVan}]}}}
        }
       ]
       decorators [
        Has Gameplay Ability {
         class=FortBTDecorator_GameplayAbility_HasGameplayAbility
         properties {gameplayAbilityTagContainer {defaultValue {gameplayTags [{tagName=Ability.Realm.ForcePullRebootVan}]}}}
        }
        Can Activate Gameplay Ability {
         class=FortBTDecorator_GameplayAbility_CanActivate
         properties {
          flowAbortMode=Both
          gameplayAbilityTagContainer {defaultValue {gameplayTags [{tagName=Ability.Realm.ForcePullRebootVan}]}}
          targetValue {key=SelfActor}
         }
        }
       ]
       properties {children [
         {childTask=&self:BTTask_RotateToFaceBBEntry_5}
         {childTask=&self:FortBTTask_ExecuteGameplayAbility_0}
        ]}
      }
      Sequence {
       class=BTComposite_Sequence
       children [
        Rotate to face BB entry {
         class=BTTask_RotateToFaceBBEntry
         decorators []
         properties {blackboardKey {selectedKeyName=AIEvaluator_Global_TargetActor}, precision {defaultValue=2}}
        }
        Leap {
         class=FortBTTask_ExecuteGameplayAbility
         decorators []
         properties {nodeName=Leap, gameplayAbilityTagContainer {defaultValue {gameplayTags [{tagName=NPCLibrary.Abilities.Leap}]}}}
        }
       ]
       decorators [
        Has Gameplay Ability {
         class=FortBTDecorator_GameplayAbility_HasGameplayAbility
         properties {gameplayAbilityTagContainer {defaultValue {gameplayTags [{tagName=NPCLibrary.Abilities.Leap}]}}}
        }
        Can Activate Gameplay Ability {
         class=FortBTDecorator_GameplayAbility_CanActivate
         properties {
          flowAbortMode=Both
          gameplayAbilityTagContainer {defaultValue {gameplayTags [{tagName=NPCLibrary.Abilities.Leap}]}}
          targetValue {key=SelfActor}
         }
        }
       ]
       properties {children [
         {childTask=&self:BTTask_RotateToFaceBBEntry_1}
         {childTask=&self:FortBTTask_ExecuteGameplayAbility_2}
        ]}
      }
      Sequence {
       class=BTComposite_Sequence
       children [
        Rotate to face BB entry {
         class=BTTask_RotateToFaceBBEntry
         decorators []
         properties {blackboardKey {selectedKeyName=AIEvaluator_Global_LastKnownPosition}, precision {defaultValue=2}}
        }
        Leap {
         class=FortBTTask_ExecuteGameplayAbility
         decorators []
         properties {nodeName=Leap, gameplayAbilityTagContainer {defaultValue {gameplayTags [{tagName=NPCLibrary.Abilities.LeapInWater}]}}}
        }
       ]
       decorators [
        Has Gameplay Ability {
         class=FortBTDecorator_GameplayAbility_HasGameplayAbility
         properties {gameplayAbilityTagContainer {defaultValue {gameplayTags [{tagName=NPCLibrary.Abilities.LeapInWater}]}}}
        }
        Can Activate Gameplay Ability {
         class=FortBTDecorator_GameplayAbility_CanActivate
         properties {
          flowAbortMode=Both
          gameplayAbilityTagContainer {defaultValue {gameplayTags [{tagName=NPCLibrary.Abilities.LeapInWater}]}}
          targetValue {key=SelfActor}
         }
        }
       ]
       properties {children [
         {childTask=&self:BTTask_RotateToFaceBBEntry_4}
         {childTask=&self:FortBTTask_ExecuteGameplayAbility_5}
        ]}
      }
      Sequence {
       class=BTComposite_Sequence
       children [
        Rotate to face BB entry {
         class=BTTask_RotateToFaceBBEntry
         decorators []
         properties {blackboardKey {selectedKeyName=AIEvaluator_Global_TargetActor}, precision {defaultValue=2}}
        }
        Throw Lightsaber {
         class=FortBTTask_ExecuteGameplayAbility
         decorators []
         properties {nodeName=Throw Lightsaber, gameplayAbilityTagContainer {defaultValue {gameplayTags [{tagName=Ability.Realm.LSThrow}]}}}
        }
       ]
       decorators [
        Has Gameplay Ability {
         class=FortBTDecorator_GameplayAbility_HasGameplayAbility
         properties {gameplayAbilityTagContainer {defaultValue {gameplayTags [{tagName=Ability.Realm.LSThrow}]}}}
        }
        Can Activate Gameplay Ability {
         class=FortBTDecorator_GameplayAbility_CanActivate
         properties {flowAbortMode=Both, gameplayAbilityTagContainer {defaultValue {gameplayTags [{tagName=Ability.Realm.LSThrow}]}}, targetValue {key=SelfActor}}
        }
       ]
       properties {children [
         {childTask=&self:BTTask_RotateToFaceBBEntry_3}
         {childTask=&self:FortBTTask_ExecuteGameplayAbility_4}
        ]}
      }
      Loop Threaten {
       class=BTComposite_Selector
       children [
        Move To Range {
         class=BTTask_RunBehavior
         decorators [
          Blackboard Based Condition {
           class=BTDecorator_Blackboard
           properties {
            arithmeticOperation=GreaterOrEqual
            flowAbortMode=None
            intValue=3
            nodeName=Blackboard Based Condition
            stringValue=EExecutionStatus::ExecutionPending
            blackboardKey {selectedKeyName=AIEvaluator_RangeAttack_ExecutionStatus}
           }
          }
          Blackboard Based Condition {
           class=BTDecorator_Blackboard
           properties {
            arithmeticOperation=GreaterOrEqual
            flowAbortMode=None
            intValue=3
            nodeName=Blackboard Based Condition
            stringValue=EExecutionStatus::ExecutionPending
            blackboardKey {selectedKeyName=AIEvaluator_ThrowableAttack_ExecutionStatus}
           }
          }
         ]
         properties {behaviorAsset=&BT_Phoebe_MoveToRange, nodeName=Move To Range}
        }
        Move To Melee {
         class=FortAthenaBTTask_BotMoveTo
         decorators [
          Can Do MeleeAttack {
           class=BTDecorator_Blackboard
           properties {
            arithmeticOperation=GreaterOrEqual
            intValue=3
            nodeName=Can Do MeleeAttack
            stringValue=EExecutionStatus::ExecutionPending
            blackboardKey {selectedKeyName=AIEvaluator_MeleeAttack_ExecutionStatus}
           }
          }
          Loop {class=BTDecorator_Loop}
         ]
         properties {
          movementResultKeyName=AIEvaluator_MeleeAttack_MovementState
          nodeName=Move To Melee
          acceptableRadius {defaultValue=5}
          allowRandomWobble {defaultValue=true}
          bAllowPartialPath {defaultValue=false}
          bAllowStrafe {defaultValue=false}
          blackboardKey {selectedKeyName=AIEvaluator_MeleeAttack_Destination}
          observedBlackboardValueTolerance {defaultValue=4}
         }
        }
        Move To Target Actor {
         class=FortAthenaBTTask_BotMoveTo
         decorators [
          Can move to TargetActor {
           class=BTDecorator_Blackboard
           properties {
            arithmeticOperation=Equal
            intValue=0
            nodeName=Can move to TargetActor
            stringValue=""
            blackboardKey {selectedKeyName=AIEvaluator_Global_TargetActor}
           }
          }
          Gameplay Tag Condition {
           class=BTDecorator_CheckGameplayTagsOnActor
           properties {
            bInverseCondition=true
            flowAbortMode=None
            nodeName=Gameplay Tag Condition
            tagsToMatch=Any
            actorToCheck {bNoneIsAllowedValue=false, selectedKeyName=SelfActor}
            gameplayTags {gameplayTags [{tagName=Ability.Realm.LSThrow.Catch}]}
           }
          }
         ]
         properties {
          movementResultKeyName=None
          nodeName=Move To Target Actor
          acceptableRadius {defaultValue=5}
          allowRandomWobble {defaultValue=true}
          bAllowPartialPath {defaultValue=false}
          bAllowStrafe {defaultValue=false}
          blackboardKey {selectedKeyName=AIEvaluator_Global_TargetActor}
          observedBlackboardValueTolerance {defaultValue=4}
         }
        }
        Fallback {class=BTTask_Wait, properties {nodeName=Fallback, waitTime {defaultValue=5}}}
       ]
       decorators [Loop {class=BTDecorator_Loop}]
       properties {
        nodeName=Loop Threaten
        children [
         {
          childComposite=null
          childTask=&self:BTTask_RunBehavior_8
          decoratorOps [{number=2, operation=Or}, {number=0}, {number=1}]
          decorators [
           &self:BTDecorator_Blackboard_49
           &self:BTDecorator_Blackboard_48
          ]
         }
         {
          childComposite=null
          childTask=&self:FortAthenaBTTask_BotMoveTo_14
          decorators [
           &self:BTDecorator_Blackboard_69
           &self:BTDecorator_Loop_15
          ]
         }
         {
          childComposite=null
          childTask=&self:FortAthenaBTTask_BotMoveTo_0
          decorators [
           &self:BTDecorator_Blackboard_7
           &self:BTDecorator_CheckGameplayTagsOnActor_0
          ]
         }
         {childComposite=null, childTask=&self:BTTask_Wait_1, decorators []}
        ]
        services []
       }
       services []
      }
     ]
     decorators [
      IsThreatened {
       class=BTDecorator_Blackboard
       properties {
        arithmeticOperation=Equal
        intValue=3
        nodeName=IsThreatened
        stringValue=EAlertLevel::Threatened
        blackboardKey {selectedKeyName=AIEvaluator_Global_AlertLevel}
       }
      }
      "Behavior Controls = CanUseThreatenedBehaviors" {
       class=FortAthenaBTDecorator_BehaviorControls
       properties {
        bInverseCondition=false
        behaviorTreeBranch=CanUseThreatenedBehaviors
        flowAbortMode=Both
        nodeName="Behavior Controls = CanUseThreatenedBehaviors"
       }
      }
     ]
     properties {
      nodeName=THREATEN
      children [
       {
        childComposite=&self:BTComposite_Sequence_2
        childTask=null
        decorators [
         &self:FortBTDecorator_GameplayAbility_HasGameplayAbility_2
         &self:FortBTDecorator_GameplayAbility_CanActivate_1
        ]
       }
       {
        childComposite=&self:BTComposite_Sequence_3
        childTask=null
        decorators [
         &self:FortBTDecorator_GameplayAbility_HasGameplayAbility_0
         &self:FortBTDecorator_GameplayAbility_CanActivate_0
        ]
       }
       {
        childComposite=&self:BTComposite_Sequence_1
        childTask=null
        decorators [
         &self:FortBTDecorator_GameplayAbility_HasGameplayAbility_1
         &self:FortBTDecorator_GameplayAbility_CanActivate_2
        ]
       }
       {
        childComposite=&self:BTComposite_Sequence_0
        childTask=null
        decorators [
         &self:FortBTDecorator_GameplayAbility_HasGameplayAbility_5
         &self:FortBTDecorator_GameplayAbility_CanActivate_5
        ]
       }
       {
        childComposite=&self:BTComposite_Sequence_4
        childTask=null
        decorators [
         &self:FortBTDecorator_GameplayAbility_HasGameplayAbility_8
         &self:FortBTDecorator_GameplayAbility_CanActivate_8
        ]
       }
       {
        childComposite=&self:BTComposite_Sequence_5
        childTask=null
        decorators [
         &self:FortBTDecorator_GameplayAbility_HasGameplayAbility_3
         &self:FortBTDecorator_GameplayAbility_CanActivate_3
        ]
       }
       {
        childComposite=&self:BTComposite_Selector_14
        childTask=null
        decorators [&self:BTDecorator_Loop_3]
       }
      ]
      services [
       &self:FortAthenaBTService_AIEvaluator_62
       &self:FortAthenaBTService_AIEvaluator_5
       &self:FortAthenaBTService_AIEvaluator_0
       &self:FortAthenaBTService_AIEvaluator_61
       &self:FortAthenaBTService_ManageWeapon_0
       &self:FortAthenaBTService_AIEvaluator_2
      ]
     }
     services [
      Evaluating...Select Weapon {
       class=FortAthenaBTService_AIEvaluator
       properties {
        interval=0.2
        nodeName=Evaluating...Select Weapon
        randomDeviation=0.05
        aIEvaluatorClass {defaultValue=&FortAthenaAIBotEvaluator_WeaponSelection}
       }
      }
      Evaluating...Melee Attack {
       class=FortAthenaBTService_AIEvaluator
       properties {
        interval=0.3
        nodeName=Evaluating...Melee Attack
        aIEvaluatorClass {defaultValue=&FortAthenaAIBotEvaluator_MeleeAttack}
       }
      }
      Evaluating...Range Attack {
       class=FortAthenaBTService_AIEvaluator
       properties {
        bCallTickOnSearchStart=true
        interval=0.2
        nodeName=Evaluating...Range Attack
        randomDeviation=0.05
        aIEvaluatorClass {defaultValue=&FortAthenaAIBotEvaluator_RangeAttack}
       }
      }
      Evaluating...Throwable Attack {
       class=FortAthenaBTService_AIEvaluator
       properties {
        interval=0.2
        nodeName=Evaluating...Throwable Attack
        randomDeviation=0.05
        aIEvaluatorClass {defaultValue=/Script/FortniteAIServer.FortAthenaAIBotEvaluator_ThrowableAttack}
       }
      }
      Executing...Weapons Actions {class=FortAthenaBTService_ManageWeapon, properties {interval=1, randomDeviation=0.3}}
      Evaluate...Sprinting {
       class=FortAthenaBTService_AIEvaluator
       properties {interval=0.2, nodeName=Evaluate...Sprinting, aIEvaluatorClass {defaultValue=&FortAthenaAIBotEvaluator_Sprinting}}
      }
     ]
    }
    Not attacking threat {
     class=BTComposite_Selector
     children [
      LKP {
       class=BTComposite_Selector
       children [Loop LKP {
         class=BTComposite_Selector
         children [
          Selector {
           class=BTComposite_Selector
           children [
            Move To LKP {
             class=FortAthenaBTTask_BotMoveTo
             decorators [Should move to LKP {
               class=BTDecorator_Blackboard
               properties {
                arithmeticOperation=Equal
                intValue=0
                nodeName=Should move to LKP
                stringValue=""
                blackboardKey {selectedKeyName=AIEvaluator_Investigate_MoveToDestination}
               }
              }]
             properties {
              movementResultKeyName=AIEvaluator_Investigate_MovementState
              nodeName=Move To LKP
              acceptableRadius {defaultValue=50}
              allowRandomWobble {defaultValue=true}
              bAllowStrafe {defaultValue=false}
              bReachTestIncludesAgentRadius {defaultValue=false}
              bReachTestIncludesGoalRadius {defaultValue=false}
              bTrackMovingGoal {defaultValue=false}
              blackboardKey {selectedKeyName=AIEvaluator_Investigate_Destination}
              observedBlackboardValueTolerance {defaultValue=4.75}
             }
            }
            Wait {class=BTTask_Wait, properties {randomDeviation {defaultValue=0.5}, waitTime {defaultValue=0.75}}}
           ]
           decorators [Should Investigate {
             class=BTDecorator_Blackboard
             properties {
              arithmeticOperation=Equal
              intValue=4
              nodeName=Should Investigate
              stringValue=EExecutionStatus::ExecutionAllowed
              blackboardKey {selectedKeyName=AIEvaluator_Investigate_ExecutionStatus}
             }
            }]
           properties {
            children [
             {
              childComposite=null
              childTask=&self:FortAthenaBTTask_BotMoveTo_3
              decorators [&self:BTDecorator_Blackboard_29]
             }
             {childComposite=null, childTask=&self:BTTask_Wait_12, decorators []}
            ]
            services []
           }
           services []
          }
          Wait {class=BTTask_Wait, properties {randomDeviation {defaultValue=0.5}, waitTime {defaultValue=0.75}}}
         ]
         decorators [Loop {class=BTDecorator_Loop}]
         properties {
          nodeName=Loop LKP
          children [
           {
            childComposite=&self:BTComposite_Selector_12
            childTask=null
            decorators [&self:BTDecorator_Blackboard_27]
           }
           {childComposite=null, childTask=&self:BTTask_Wait_9, decorators []}
          ]
          services []
         }
         services []
        }]
       decorators [
        IsLKP {
         class=BTDecorator_Blackboard
         properties {
          arithmeticOperation=Equal
          intValue=2
          nodeName=IsLKP
          stringValue=EAlertLevel::LKP
          blackboardKey {selectedKeyName=AIEvaluator_Global_AlertLevel}
         }
        }
        "Behavior Controls = CanUseLastKnownPositionBehavior" {
         class=FortAthenaBTDecorator_BehaviorControls
         properties {
          bInverseCondition=false
          behaviorTreeBranch=CanUseLastKnownPositionBehavior
          flowAbortMode=Both
          nodeName="Behavior Controls = CanUseLastKnownPositionBehavior"
         }
        }
       ]
       properties {
        nodeName=LKP
        children [{
          childComposite=&self:BTComposite_Selector_10
          childTask=null
          decorators [&self:BTDecorator_Loop_5]
         }]
        services [
         &self:FortAthenaBTService_AIEvaluator_4
         &self:FortAthenaBTService_AIEvaluator_31
         &self:FortAthenaBTService_AIEvaluator_32
         &self:FortAthenaBTService_AIEvaluator_35
         &self:FortAthenaBTService_ManageWeapon_4
         &self:FortAthenaBTService_AIEvaluator_36
        ]
       }
       services [
        Evaluating...Select Weapon {
         class=FortAthenaBTService_AIEvaluator
         properties {
          interval=0.2
          nodeName=Evaluating...Select Weapon
          randomDeviation=0.05
          aIEvaluatorClass {defaultValue=&FortAthenaAIBotEvaluator_WeaponSelection}
         }
        }
        Evaluating...Investigate Treat {
         class=FortAthenaBTService_AIEvaluator
         properties {
          bCallTickOnSearchStart=true
          interval=0.5
          nodeName=Evaluating...Investigate Treat
          aIEvaluatorClass {defaultValue=/Script/FortniteAIServer.FortAthenaAIBotEvaluator_Investigate}
         }
        }
        Evaluating...Melee Attack {
         class=FortAthenaBTService_AIEvaluator
         properties {
          interval=0.3
          nodeName=Evaluating...Melee Attack
          aIEvaluatorClass {defaultValue=&FortAthenaAIBotEvaluator_MeleeAttack}
         }
        }
        Evaluating...Range Attack {
         class=FortAthenaBTService_AIEvaluator
         properties {
          interval=0.3
          nodeName=Evaluating...Range Attack
          aIEvaluatorClass {defaultValue=&FortAthenaAIBotEvaluator_RangeAttack}
         }
        }
        Executing...Weapons Actions {class=FortAthenaBTService_ManageWeapon, properties {interval=0.2, randomDeviation=0.1}}
        Evaluate...Sprinting {
         class=FortAthenaBTService_AIEvaluator
         properties {interval=0.2, nodeName=Evaluate...Sprinting, aIEvaluatorClass {defaultValue=&FortAthenaAIBotEvaluator_Sprinting}}
        }
       ]
      }
      ALERTED {
       class=BTComposite_Selector
       children [Loop Alerted {
         class=BTComposite_Selector
         children [
          Wait {
           class=BTTask_Wait
           decorators [Should Observe {
             class=BTDecorator_Blackboard
             properties {
              arithmeticOperation=GreaterOrEqual
              intValue=3
              nodeName=Should Observe
              stringValue=EExecutionStatus::ExecutionPending
              blackboardKey {selectedKeyName=AIEvaluator_Observe_ExecutionStatus}
             }
            }]
           properties {randomDeviation {defaultValue=0.25}, waitTime {defaultValue=1}}
          }
          FallBack {class=BTTask_Wait, properties {nodeName=FallBack, waitTime {defaultValue=1}}}
         ]
         decorators [Loop {class=BTDecorator_Loop}]
         properties {
          nodeName=Loop Alerted
          children [
           {
            childComposite=null
            childTask=&self:BTTask_Wait_2
            decorators [&self:BTDecorator_Blackboard_6]
           }
           {childComposite=null, childTask=&self:BTTask_Wait_8, decorators []}
          ]
          services []
         }
         services []
        }]
       decorators [
        IsAlerted {
         class=BTDecorator_Blackboard
         properties {
          arithmeticOperation=Equal
          intValue=1
          nodeName=IsAlerted
          stringValue=EAlertLevel::Alerted
          blackboardKey {selectedKeyName=AIEvaluator_Global_AlertLevel}
         }
        }
        "Behavior Controls = CanUseAlertedBehavior" {
         class=FortAthenaBTDecorator_BehaviorControls
         properties {bInverseCondition=false, behaviorTreeBranch=CanUseAlertedBehavior, flowAbortMode=Both, nodeName="Behavior Controls = CanUseAlertedBehavior"}
        }
       ]
       properties {
        nodeName=ALERTED
        children [{
          childComposite=&self:BTComposite_Selector_1
          childTask=null
          decorators [&self:BTDecorator_Loop_6]
         }]
        services [
         &self:FortAthenaBTService_AIEvaluator_7
         &self:FortAthenaBTService_AIEvaluator_8
        ]
       }
       services [
        Evaluating...Observe Treat {
         class=FortAthenaBTService_AIEvaluator
         properties {
          bCallTickOnSearchStart=true
          interval=0.5
          nodeName=Evaluating...Observe Treat
          aIEvaluatorClass {defaultValue=/Script/FortniteAIServer.FortAthenaAIBotEvaluator_Observe}
         }
        }
        Evaluate...Sprinting {
         class=FortAthenaBTService_AIEvaluator
         properties {interval=0.2, nodeName=Evaluate...Sprinting, aIEvaluatorClass {defaultValue=&FortAthenaAIBotEvaluator_Sprinting}}
        }
       ]
      }
      Run Behavior {
       class=BTTask_RunBehavior
       decorators [
        Revive {
         class=BTDecorator_Blackboard
         properties {
          arithmeticOperation=GreaterOrEqual
          intValue=3
          nodeName=Revive
          stringValue=EExecutionStatus::ExecutionPending
          blackboardKey {selectedKeyName=AIEvaluator_Revive_ExecutionStatus}
         }
        }
        "Behavior Controls = CanReviveDBNOTeammates" {
         class=FortAthenaBTDecorator_BehaviorControls
         properties {
          bInverseCondition=false
          behaviorTreeBranch=CanReviveDBNOTeammates
          flowAbortMode=Both
          nodeName="Behavior Controls = CanReviveDBNOTeammates"
         }
        }
       ]
       properties {behaviorAsset=/Game/Athena/AI/Phoebe/BehaviorTrees/BT_Phoebe_Revive, nodeName=Run Behavior}
      }
      UNAWARE {
       class=BTComposite_Selector
       children [Loop Unaware {
         class=BTComposite_Selector
         children [
          PlayEmote {
           class=FortAthenaBTTask_PlayEmote
           children []
           decorators [
            ForcePlayEmote? {
             class=BTDecorator_Blackboard
             properties {arithmeticOperation=Equal, intValue=0, nodeName=ForcePlayEmote?, stringValue="", blackboardKey {selectedKeyName=AIEvaluator_ForcePlayEmote}}
            }
            "Behavior Controls = CanPlayEmote" {
             class=FortAthenaBTDecorator_BehaviorControls
             properties {bInverseCondition=false, behaviorTreeBranch=CanPlayEmote, flowAbortMode=Both, nodeName="Behavior Controls = CanPlayEmote"}
            }
           ]
           properties {bIgnoreRestartSelf=false, nodeName=PlayEmote, playEmoteExecutionStatusKeyName=AIEvaluator_PlayEmote_ExecutionStatus, bespokeEmotes []}
           services []
          }
          "Converted Patrol " {
           class=BTComposite_Selector
           children [
            GameplayAbility {
             class=FortBTTask_ExecuteGameplayAbility
             decorators [
              Should Teleport Towards Converter? {
               class=BTDecorator_Blackboard
               properties {
                arithmeticOperation=Equal
                intValue=0
                nodeName=Should Teleport Towards Converter?
                stringValue=""
                blackboardKey {selectedKeyName=AIEvaluator_Converted_ShouldTeleportTowardsConverter}
               }
              }
              Has Converted Teleport Ability? {
               class=FortBTDecorator_GameplayAbility_HasGameplayAbility
               properties {nodeName=Has Converted Teleport Ability?, gameplayAbilityTagContainer {defaultValue {gameplayTags [{tagName=Ability.Converted.Teleport}]}}}
              }
              Can Activate Converted Teleport Ability? {
               class=FortBTDecorator_GameplayAbility_CanActivate
               properties {
                flowAbortMode=Both
                nodeName=Can Activate Converted Teleport Ability?
                gameplayAbilityTagContainer {defaultValue {gameplayTags [{tagName=Ability.Converted.Teleport}]}}
                targetValue {key=SelfActor}
               }
              }
             ]
             properties {nodeName=GameplayAbility, gameplayAbilityTagContainer {defaultValue {gameplayTags [{tagName=Ability.Converted.Teleport}]}}}
            }
            Injection Point Athena.Branch.Unaware.Converted.HighPriority {
             class=FortBTTask_InjectionPoint
             children []
             decorators [Can Do Injected Behavior {
               class=FortBTDecorator_CanDoInjectedBehavior
               properties {
                arithmeticOperation=GreaterOrEqual
                bInverseCondition=false
                basicOperation=Set
                floatValue=0
                flowAbortMode=Both
                intValue=3
                nodeName=Can Do Injected Behavior
                notifyObserver=ResultChange
                stringValue=ExecutionPending
                textOperation=Equal
                blackboardKey {bNoneIsAllowedValue=false, selectedKeyName=AIEvaluator_Injection_Athena_Branch_Unaware_Converted_HighPriority}
               }
              }]
             properties {
              bIgnoreRestartSelf=false
              nodeName=Injection Point Athena.Branch.Unaware.Converted.HighPriority
              injectedAction {baseClass=&Object, defaultValue=null, key=Injection_Athena_Branch_Unaware_Converted_HighPriority_ActionObject}
              injectionTag {tagName=Behavior.Injection.Athena.Branch.Unaware.Converted.HighPriority}
             }
             services []
            }
            Conversation {
             class=FortAthenaBTTask_Conversation
             children []
             decorators [
              Conversation? {
               class=BTDecorator_Blackboard
               properties {
                arithmeticOperation=GreaterOrEqual
                intValue=3
                nodeName=Conversation?
                stringValue=EExecutionStatus::ExecutionPending
                blackboardKey {selectedKeyName=AIEvaluator_Conversation_ExecutionStatus}
               }
              }
              "Behavior Controls = CanConverse" {
               class=FortAthenaBTDecorator_BehaviorControls
               properties {bInverseCondition=false, behaviorTreeBranch=CanConverse, flowAbortMode=Both, nodeName="Behavior Controls = CanConverse"}
              }
             ]
             properties {
              bIgnoreRestartSelf=false
              bResetEvaluatorStatusKeyOnFinish=true
              conversationStatusKeyName=AIEvaluator_Conversation_ExecutionStatus
              nodeName=Conversation
             }
             services []
            }
            UseSmartObject {
             class=BTTask_RunBehavior
             decorators [
              Use Smart Object? {
               class=BTDecorator_Blackboard
               properties {
                arithmeticOperation=GreaterOrEqual
                intValue=3
                nodeName=Use Smart Object?
                stringValue=EExecutionStatus::ExecutionPending
                blackboardKey {selectedKeyName=AIEvaluator_SmartObject_ExecutionStatus}
               }
              }
              "Behavior Controls = CanUseSmartObjects" {
               class=FortAthenaBTDecorator_BehaviorControls
               properties {bInverseCondition=false, behaviorTreeBranch=CanUseSmartObjects, flowAbortMode=Both, nodeName="Behavior Controls = CanUseSmartObjects"}
              }
             ]
             properties {behaviorAsset=/Game/Athena/AI/Phoebe/BehaviorTrees/BT_Phoebe_UseSmartObject, nodeName=UseSmartObject}
            }
            Selector {
             class=BTComposite_Selector
             children [
              Selector {
               class=BTComposite_Selector
               children [
                Run Unstuck Behavior {
                 class=BTTask_RunBehavior
                 decorators []
                 properties {behaviorAsset=&BT_Phoebe_UnstuckNew, nodeName=Run Unstuck Behavior}
                }
                Sequence {
                 class=BTComposite_Sequence
                 children [
                  Rotate to face BB entry {
                   class=BTTask_RotateToFaceBBEntry
                   decorators []
                   properties {blackboardKey {selectedKeyName=AIEvaluator_Converted_Destination}, precision {defaultValue=2}}
                  }
                  Leap {
                   class=FortBTTask_ExecuteGameplayAbility
                   decorators []
                   properties {nodeName=Leap, gameplayAbilityTagContainer {defaultValue {gameplayTags [{tagName=Jethro.HeavyLamp.Abilities.CatchUpLeap}]}}}
                  }
                  Wait {class=BTTask_Wait, properties {randomDeviation {defaultValue=0.5}, waitTime {defaultValue=1.5}}}
                 ]
                 decorators [
                  Has Gameplay Ability {
                   class=FortBTDecorator_GameplayAbility_HasGameplayAbility
                   properties {gameplayAbilityTagContainer {defaultValue {gameplayTags [{tagName=Jethro.HeavyLamp.Abilities.CatchUpLeap}]}}}
                  }
                  Can Activate Gameplay Ability {
                   class=FortBTDecorator_GameplayAbility_CanActivate
                   properties {
                    flowAbortMode=LowerPriority
                    gameplayAbilityTagContainer {defaultValue {gameplayTags [{tagName=Jethro.HeavyLamp.Abilities.CatchUpLeap}]}}
                    targetValue {key=SelfActor}
                   }
                  }
                 ]
                 properties {children [
                   {childTask=&self:BTTask_RotateToFaceBBEntry_8}
                   {childTask=&self:FortBTTask_ExecuteGameplayAbility_8}
                   {childTask=&self:BTTask_Wait_13}
                  ]}
                }
                Move To Converter {
                 class=FortAthenaBTTask_BotMoveTo
                 decorators []
                 properties {
                  movementResultKeyName=AIEvaluator_Converted_MovementState
                  nodeName=Move To Converter
                  acceptableRadius {defaultValue=50}
                  allowRandomWobble {defaultValue=true}
                  bAllowStrafe {defaultValue=false}
                  bReachTestIncludesAgentRadius {defaultValue=false}
                  bReachTestIncludesGoalRadius {defaultValue=false}
                  bTrackMovingGoal {defaultValue=false}
                  blackboardKey {selectedKeyName=AIEvaluator_Converted_Destination}
                  observedBlackboardValueTolerance {defaultValue=4.75}
                 }
                }
               ]
               decorators [Should Move Towards Converter? {
                 class=BTDecorator_Blackboard
                 properties {
                  arithmeticOperation=Equal
                  intValue=4
                  nodeName=Should Move Towards Converter?
                  stringValue=""
                  blackboardKey {selectedKeyName=AIEvaluator_Converted_ShouldMoveTowardsConverter}
                 }
                }]
               properties {
                children [
                 {childComposite=null, childTask=&self:BTTask_RunBehavior_6, decorators []}
                 {
                  childComposite=&self:BTComposite_Sequence_10
                  childTask=null
                  decorators [
                   &self:FortBTDecorator_GameplayAbility_HasGameplayAbility_11
                   &self:FortBTDecorator_GameplayAbility_CanActivate_11
                  ]
                 }
                 {childComposite=null, childTask=&self:FortAthenaBTTask_BotMoveTo_12, decorators []}
                ]
                services []
               }
               services []
              }
              Sequence {
               class=BTComposite_Sequence
               children [
                Run PatrolAround Behavior {
                 class=BTTask_RunBehavior
                 decorators [Patrol Around {
                   class=BTDecorator_Blackboard
                   properties {
                    arithmeticOperation=GreaterOrEqual
                    flowAbortMode=None
                    intValue=2
                    nodeName=Patrol Around
                    stringValue=EExecutionStatus::ExecutionSuccess
                    blackboardKey {selectedKeyName=AIEvaluator_Patrol_ExecutionStatus}
                   }
                  }]
                 properties {behaviorAsset=/Game/Athena/AI/Phoebe/BehaviorTrees/BT_Phoebe_PatrolAround, nodeName=Run PatrolAround Behavior}
                }
                Wait {class=BTTask_Wait, properties {randomDeviation {defaultValue=1}, waitTime {defaultValue=3}}}
               ]
               decorators [Allowed To Patrol Around? {
                 class=BTDecorator_Blackboard
                 properties {
                  arithmeticOperation=Equal
                  intValue=0
                  nodeName=Allowed To Patrol Around?
                  stringValue=""
                  blackboardKey {selectedKeyName=AIEvaluator_Converted_AllowPatrolAround}
                 }
                }]
               properties {
                children [
                 {
                  childTask=&self:BTTask_RunBehavior_11
                  decorators [&self:BTDecorator_Blackboard_56]
                 }
                 {childTask=&self:BTTask_Wait_29}
                ]
                services [
                 &self:FortAthenaBTService_AIEvaluator_73
                 &self:FortAthenaBTService_AIEvaluator_74
                ]
               }
               services [
                Evaluating... Patrol Around {
                 class=FortAthenaBTService_AIEvaluator
                 properties {
                  bCallTickOnSearchStart=true
                  interval=2.5
                  nodeName=Evaluating... Patrol Around
                  randomDeviation=0.5
                  aIEvaluatorClass {defaultValue=/Game/Athena/AI/NPCs/Base/Evaluators/Evaluator_PatrolAround_NPC.Evaluator_PatrolAround_NPC_C}
                 }
                }
                Evaluating... SprintingInWaterOnly {
                 class=FortAthenaBTService_AIEvaluator
                 properties {
                  interval=0.2
                  nodeName=Evaluating... SprintingInWaterOnly
                  aIEvaluatorClass {defaultValue=/Game/Athena/AI/NPCs/Base/Evaluators/Evaluator_Sprinting_InWaterOnly.Evaluator_Sprinting_InWaterOnly_C}
                 }
                }
               ]
              }
             ]
             decorators []
             properties {
              children [
               {
                childComposite=&self:BTComposite_Selector_21
                childTask=null
                decorators [&self:BTDecorator_Blackboard_59]
               }
               {
                childComposite=&self:BTComposite_Sequence_11
                childTask=null
                decorators [&self:BTDecorator_Blackboard_55]
               }
              ]
              services [&self:FortAthenaBTService_AIEvaluator_72]
             }
             services [Evaluating...Melee Attack {
               class=FortAthenaBTService_AIEvaluator
               properties {
                interval=0.3
                nodeName=Evaluating...Melee Attack
                aIEvaluatorClass {defaultValue=&FortAthenaAIBotEvaluator_MeleeAttack}
               }
              }]
            }
            Wait {
             class=BTTask_Wait
             decorators [Can Scan Around When Waiting? {
               class=BTDecorator_Blackboard
               properties {
                arithmeticOperation=Equal
                intValue=0
                nodeName=Can Scan Around When Waiting?
                stringValue=""
                blackboardKey {selectedKeyName=AIEvaluator_Converted_AllowScanAroundWhenWaiting}
               }
              }]
             properties {waitTime {defaultValue=2}}
            }
            Selector {
             class=BTComposite_Selector
             children [
              Wait {
               class=BTTask_Wait
               decorators [Blackboard Based Condition {
                 class=BTDecorator_Blackboard
                 properties {
                  arithmeticOperation=Equal
                  flowAbortMode=Self
                  intValue=0
                  nodeName=Blackboard Based Condition
                  notifyObserver=ValueChange
                  stringValue=""
                  blackboardKey {selectedKeyName=HiredPlayerFocusLocation}
                 }
                }]
               properties {waitTime {defaultValue=2}}
              }
              Wait {class=BTTask_Wait, properties {waitTime {defaultValue=2}}}
             ]
             decorators []
             properties {
              children [
               {
                childComposite=null
                childTask=&self:BTTask_Wait_11
                decorators [&self:BTDecorator_Blackboard_18]
               }
               {childComposite=null, childTask=&self:BTTask_Wait_3, decorators []}
              ]
              services []
             }
             services []
            }
           ]
           decorators [
            Is Converted? {
             class=BTDecorator_GameplayTagQuery
             properties {
              flowAbortMode=Both
              nodeName=Is Converted?
              gameplayTagQuery {
               autoDescription=" ANY( Athena.Gameplay.Status.Converted )"
               queryTokenStream [0,1,1,1,0]
               tagDictionary [{tagName=Athena.Gameplay.Status.Converted}]
              }
             }
            }
            "Behavior Controls = CanBeConverted" {
             class=FortAthenaBTDecorator_BehaviorControls
             properties {bInverseCondition=false, behaviorTreeBranch=CanBeConverted, flowAbortMode=Both, nodeName="Behavior Controls = CanBeConverted"}
            }
           ]
           properties {
            nodeName="Converted Patrol "
            children [
             {
              childComposite=null
              childTask=&self:FortBTTask_ExecuteGameplayAbility_9
              decorators [
               &self:BTDecorator_Blackboard_41
               &self:FortBTDecorator_GameplayAbility_HasGameplayAbility_12
               &self:FortBTDecorator_GameplayAbility_CanActivate_15
              ]
             }
             {
              childComposite=null
              childTask=&self:FortBTTask_InjectionPoint_0
              decorators [&self:FortBTDecorator_CanDoInjectedBehavior_1]
             }
             {
              childComposite=null
              childTask=&self:FortAthenaBTTask_Conversation_2
              decorators [
               &self:BTDecorator_Blackboard_34
               &self:FortAthenaBTDecorator_BehaviorControls_12
              ]
             }
             {
              childComposite=null
              childTask=&self:BTTask_RunBehavior_0
              decorators [
               &self:BTDecorator_Blackboard_46
               &self:FortAthenaBTDecorator_BehaviorControls_20
              ]
             }
             {childComposite=&self:BTComposite_Selector_9, childTask=null, decorators []}
             {
              childComposite=null
              childTask=&self:BTTask_Wait_7
              decorators [&self:BTDecorator_Blackboard_24]
             }
             {childComposite=&self:BTComposite_Selector_19, childTask=null, decorators []}
            ]
            services [
             &self:FortAthenaBTService_AIEvaluator_57
             &self:FortBTService_AIEvaluatorsInjector_0
             &self:BTService_EvaluateHiredPlayerFocus_C_0
            ]
           }
           services [
            Evaluating... Converted Actions {
             class=FortAthenaBTService_AIEvaluator
             properties {
              interval=0.5
              nodeName=Evaluating... Converted Actions
              aIEvaluatorClass {defaultValue=/Game/Athena/AI/NPCs/Base/Evaluators/Evaluator_Converted_NPC.Evaluator_Converted_NPC_C}
             }
            }
            Behavior Injector {
             class=FortBTService_AIEvaluatorsInjector
             properties {
              bCallTickOnSearchStart=false
              bRestartTimerOnEachActivation=false
              interval=0.5
              nodeName=Behavior Injector
              randomDeviation=0.1
              injectionTagsKeys [{
                actionInterfaceObjectKeyName=Injection_Athena_Branch_Unaware_Converted_HighPriority_ActionObject
                injectionKeyName=AIEvaluator_Injection_Athena_Branch_Unaware_Converted_HighPriority
                injectionTag {tagName=Behavior.Injection.Athena.Branch.Unaware.Converted.HighPriority}
               }]
             }
            }
            BTService_EvaluateHiredPlayerFocus_C_0 {
             class=BTService_EvaluateHiredPlayerFocus_C
             properties {
              bCallTickOnSearchStart=false
              bRestartTimerOnEachActivation=false
              bShowEventDetails=false
              bShowPropertyDetails=true
              customDescription=""
              goalActorKeyName=GoalActor
              interval=0.5
              nodeName=""
              randomDeviation=0.1
              playerFocusKey {bNoneIsAllowedValue=false, selectedKeyName=AIEvaluator_Global_LastKnownPosition}
             }
            }
           ]
          }
          UseSmartObject {
           class=BTTask_RunBehavior
           decorators []
           properties {behaviorAsset=&BT_Phoebe_UseSmartObject, nodeName=UseSmartObject}
          }
          Run DynamicBlueprint {
           class=FortAthenaBTTask_DynamicBlueprint
           children []
           decorators [Can Run LocalGameplayBehavior? {
             class=BTDecorator_Blackboard
             properties {
              arithmeticOperation=GreaterOrEqual
              intValue=3
              nodeName=Can Run LocalGameplayBehavior?
              stringValue=EExecutionStatus::ExecutionPending
              blackboardKey {selectedKeyName=AIEvaluator_DynamicBlueprint_ExecutionStatus}
             }
            }]
           properties {
            bIgnoreRestartSelf=false
            dynamicBlueprintActorKeyName=AIEvaluator_DynamicBlueprint_Actor
            dynamicBlueprintStatusKeyName=AIEvaluator_DynamicBlueprint_ExecutionStatus
            nodeName=Run DynamicBlueprint
           }
           services []
          }
          Can Patrolling? {
           class=BTComposite_Selector
           children [
            Selector {
             class=BTComposite_Selector
             children [
              Run Unstuck Behavior {
               class=BTTask_RunBehavior
               decorators []
               properties {behaviorAsset=&BT_Phoebe_UnstuckNew, nodeName=Run Unstuck Behavior}
              }
              MoveTo PatrolPoint {
               class=FortAthenaBTTask_BotMoveTo
               decorators [Is Stressed? {
                 class=BTDecorator_Blackboard
                 properties {
                  arithmeticOperation=GreaterOrEqual
                  intValue=3
                  nodeName=Is Stressed?
                  stringValue=""
                  blackboardKey {selectedKeyName="AIEvaluator_IsStressed "}
                 }
                }]
               properties {
                movementResultKeyName=AIEvaluator_Patrolling_MovementState
                nodeName=MoveTo PatrolPoint
                acceptableRadius {defaultValue=10}
                allowRandomWobble {defaultValue=true}
                bAllowStrafe {defaultValue=true}
                bTrackMovingGoal {defaultValue=false}
                blackboardKey {selectedKeyName=AIEvaluator_Patrolling_Destination}
                observedBlackboardValueTolerance {defaultValue=50}
               }
              }
              MoveTo PatrolPoint {
               class=FortAthenaBTTask_BotMoveTo
               decorators []
               properties {
                movementResultKeyName=AIEvaluator_Patrolling_MovementState
                nodeName=MoveTo PatrolPoint
                acceptableRadius {defaultValue=10}
                allowRandomWobble {defaultValue=true}
                bAllowStrafe {defaultValue=true}
                bTrackMovingGoal {defaultValue=false}
                blackboardKey {selectedKeyName=AIEvaluator_Patrolling_Destination}
                observedBlackboardValueTolerance {defaultValue=50}
               }
              }
             ]
             decorators [Should move to patrol point? {
               class=BTDecorator_Blackboard
               properties {
                arithmeticOperation=Equal
                intValue=0
                nodeName=Should move to patrol point?
                stringValue=""
                blackboardKey {selectedKeyName=AIEvaluator_Patrolling_ShouldMove}
               }
              }]
             properties {
              children [
               {childComposite=null, childTask=&self:BTTask_RunBehavior_2, decorators []}
               {
                childComposite=null
                childTask=&self:FortAthenaBTTask_BotMoveTo_10
                decorators [&self:BTDecorator_Blackboard_26]
               }
               {childComposite=null, childTask=&self:FortAthenaBTTask_BotMoveTo_9, decorators []}
              ]
              services []
             }
             services []
            }
            Wait On PatrolPoint {class=BTTask_Wait, properties {nodeName=Wait On PatrolPoint, waitTime {defaultValue=5}}}
           ]
           decorators [
            Can Patrolling? {
             class=BTDecorator_Blackboard
             properties {
              arithmeticOperation=GreaterOrEqual
              intValue=3
              nodeName=Can Patrolling?
              stringValue=EExecutionStatus::ExecutionPending
              blackboardKey {selectedKeyName=AIEvaluator_Patrolling_ExecutionStatus}
             }
            }
            "Behavior Controls = CanPatrolOnPath" {
             class=FortAthenaBTDecorator_BehaviorControls
             properties {bInverseCondition=false, behaviorTreeBranch=CanPatrolOnPath, flowAbortMode=Both, nodeName="Behavior Controls = CanPatrolOnPath"}
            }
           ]
           properties {
            nodeName=Can Patrolling?
            children [
             {
              childComposite=&self:BTComposite_Selector_13
              childTask=null
              decorators [&self:BTDecorator_Blackboard_30]
             }
             {childComposite=null, childTask=&self:BTTask_Wait_14, decorators []}
            ]
            services []
           }
           services []
          }
          Fallback Behavior {
           class=BTComposite_Selector
           children [
            Run PatrolAround Behavior {
             class=BTTask_RunBehavior
             decorators [Patrol Around {
               class=BTDecorator_Blackboard
               properties {
                arithmeticOperation=GreaterOrEqual
                intValue=2
                nodeName=Patrol Around
                stringValue=EExecutionStatus::ExecutionSuccess
                blackboardKey {selectedKeyName=AIEvaluator_Patrol_ExecutionStatus}
               }
              }]
             properties {behaviorAsset=&BT_Phoebe_PatrolAround, nodeName=Run PatrolAround Behavior}
            }
            Wait {class=BTTask_Wait, properties {waitTime {defaultValue=2}}}
           ]
           decorators [
            Can Patrol Around? {
             class=BTDecorator_Blackboard
             properties {
              arithmeticOperation=Equal
              intValue=0
              nodeName=Can Patrol Around?
              stringValue=""
              blackboardKey {selectedKeyName=AIEvaluator_CanUseFallbackPatrolAround}
             }
            }
            "Behavior Controls = CanPatrolAround" {
             class=FortAthenaBTDecorator_BehaviorControls
             properties {bInverseCondition=false, behaviorTreeBranch=CanPatrolAround, flowAbortMode=Both, nodeName="Behavior Controls = CanPatrolAround"}
            }
           ]
           properties {
            nodeName=Fallback Behavior
            children [
             {
              childComposite=null
              childTask=&self:BTTask_RunBehavior_3
              decorators [&self:BTDecorator_Blackboard_32]
             }
             {childComposite=null, childTask=&self:BTTask_Wait_15, decorators []}
            ]
            services [
             &self:FortAthenaBTService_AIEvaluator_17
             &self:FortAthenaBTService_AIEvaluator_37
            ]
           }
           services [
            Evaluating... Patrol Around {
             class=FortAthenaBTService_AIEvaluator
             properties {
              bCallTickOnSearchStart=true
              interval=2.5
              nodeName=Evaluating... Patrol Around
              randomDeviation=0.5
              aIEvaluatorClass {defaultValue=/Script/FortniteAIServer.FortAthenaAIBotEvaluator_PatrolAround}
             }
            }
            Focus on target {
             class=FortAthenaBTService_AIEvaluator
             properties {interval=0.5, nodeName=Focus on target, aIEvaluatorClass {defaultValue=&FortAthenaAIBotEvaluator_HandleFocusing}}
            }
           ]
          }
          Wait {class=BTTask_Wait, properties {waitTime {defaultValue=2}}}
         ]
         decorators [Loop {class=BTDecorator_Loop}]
         properties {
          nodeName=Loop Unaware
          children [
           {
            childComposite=null
            childTask=&self:FortAthenaBTTask_PlayEmote_3
            decorators [
             &self:BTDecorator_Blackboard_8
             &self:FortAthenaBTDecorator_BehaviorControls_5
            ]
           }
           {
            childComposite=&self:BTComposite_Selector_0
            childTask=null
            decorators [
             &self:BTDecorator_GameplayTagQuery_1
             &self:FortAthenaBTDecorator_BehaviorControls_11
            ]
           }
           {childComposite=null, childTask=&self:BTTask_RunBehavior_5, decorators []}
           {
            childComposite=null
            childTask=&self:FortAthenaBTTask_DynamicBlueprint_1
            decorators [&self:BTDecorator_Blackboard_11]
           }
           {
            childComposite=&self:BTComposite_Selector_18
            childTask=null
            decorators [
             &self:BTDecorator_Blackboard_33
             &self:FortAthenaBTDecorator_BehaviorControls_13
            ]
           }
           {
            childComposite=&self:BTComposite_Selector_17
            childTask=null
            decorators [
             &self:BTDecorator_Blackboard_31
             &self:FortAthenaBTDecorator_BehaviorControls_14
            ]
           }
           {childComposite=null, childTask=&self:BTTask_Wait_16, decorators []}
          ]
          services []
         }
         services []
        }]
       decorators []
       properties {
        nodeName=UNAWARE
        children [{
          childComposite=&self:BTComposite_Selector_25
          childTask=null
          decorators [&self:BTDecorator_Loop_9]
         }]
        services [
         &self:FortAthenaBTService_AIEvaluator_14
         &self:FortAthenaBTService_AIEvaluator_46
         &self:FortAthenaBTService_AIEvaluator_47
         &self:FortAthenaBTService_AIEvaluator_48
         &self:FortAthenaBTService_AIEvaluator_50
         &self:FortAthenaBTService_AIEvaluator_10
         &self:FortAthenaBTService_ManageWeapon_9
         &self:FortAthenaBTService_AIEvaluator_3
         &self:FortAthenaBTService_AIEvaluator_15
         &self:FortAthenaBTService_AIEvaluator_84
        ]
       }
       services [
        Evaluating...Select Weapon {
         class=FortAthenaBTService_AIEvaluator
         properties {
          interval=0.2
          nodeName=Evaluating...Select Weapon
          randomDeviation=0.05
          aIEvaluatorClass {defaultValue=&FortAthenaAIBotEvaluator_WeaponSelection}
         }
        }
        Evaluating...Patrolling {
         class=FortAthenaBTService_AIEvaluator
         properties {
          interval=0.5
          nodeName=Evaluating...Patrolling
          aIEvaluatorClass {defaultValue=/Game/Athena/AI/NPCs/Base/Evaluators/Evaluator_Patrolling_NPC.Evaluator_Patrolling_NPC_C}
         }
        }
        Evaluating...Melee Attack {
         class=FortAthenaBTService_AIEvaluator
         properties {
          interval=0.3
          nodeName=Evaluating...Melee Attack
          aIEvaluatorClass {defaultValue=&FortAthenaAIBotEvaluator_MeleeAttack}
         }
        }
        Evaluating...Range Attack {
         class=FortAthenaBTService_AIEvaluator
         properties {
          interval=1
          nodeName=Evaluating...Range Attack
          aIEvaluatorClass {defaultValue=&FortAthenaAIBotEvaluator_RangeAttack}
         }
        }
        Evaluating...StressTag {
         class=FortAthenaBTService_AIEvaluator
         properties {
          interval=0.5
          nodeName=Evaluating...StressTag
          aIEvaluatorClass {defaultValue=/Game/Athena/AI/NPCs/Base/Evaluators/Evaluator_StressTag_Guard.Evaluator_StressTag_Guard_C}
         }
        }
        Evaluating... Holster Weapon {
         class=FortAthenaBTService_AIEvaluator
         properties {
          interval=1
          nodeName=Evaluating... Holster Weapon
          randomDeviation=0.2
          aIEvaluatorClass {defaultValue=/Script/FortniteAIServer.FortAthenaAIBotEvaluator_HolsterWeapon}
         }
        }
        Executing...Weapons Actions {class=FortAthenaBTService_ManageWeapon, properties {interval=0.2, randomDeviation=0.1}}
        Evaluating...Conversation {
         class=FortAthenaBTService_AIEvaluator
         properties {
          interval=1
          nodeName=Evaluating...Conversation
          randomDeviation=0.2
          aIEvaluatorClass {defaultValue=/Script/FortniteAIServer.FortAthenaAIBotEvaluator_Conversation}
         }
        }
        Evaluating...SmartObject {
         class=FortAthenaBTService_AIEvaluator
         properties {
          interval=2
          nodeName=Evaluating...SmartObject
          aIEvaluatorClass {defaultValue=/Script/FortniteAIServer.FortAthenaAIBotEvaluator_SmartObjects}
         }
        }
        Evaluating... FollowGroupLeader {
         class=FortAthenaBTService_AIEvaluator
         properties {
          interval=0.5
          nodeName=Evaluating... FollowGroupLeader
          aIEvaluatorClass {defaultValue=/Script/FortniteAIServer.FortAthenaAIEvaluator_FollowGroupLeader}
         }
        }
       ]
      }
      Fallback {class=BTTask_Wait, properties {nodeName=Fallback, waitTime {defaultValue=5}}}
     ]
     decorators []
     properties {
      nodeName=Not attacking threat
      children [
       {
        childComposite=&self:BTComposite_Selector_8
        childTask=null
        decorators [
         &self:BTDecorator_Blackboard_9
         &self:FortAthenaBTDecorator_BehaviorControls_7
        ]
       }
       {
        childComposite=&self:BTComposite_Selector_3
        childTask=null
        decorators [
         &self:BTDecorator_Blackboard_1
         &self:FortAthenaBTDecorator_BehaviorControls_8
        ]
       }
       {
        childComposite=null
        childTask=&self:BTTask_RunBehavior_7
        decorators [
         &self:BTDecorator_Blackboard_17
         &self:FortAthenaBTDecorator_BehaviorControls_9
        ]
       }
       {childComposite=&self:BTComposite_Selector_26, childTask=null, decorators []}
       {childComposite=null, childTask=&self:BTTask_Wait_6, decorators []}
      ]
      services [&self:FortAthenaBTService_AIEvaluator_16]
     }
     services [Evaluating...Revive {
       class=FortAthenaBTService_AIEvaluator
       properties {
        interval=1
        nodeName=Evaluating...Revive
        aIEvaluatorClass {defaultValue=/Game/Athena/AI/NPCs/Base/Evaluators/Evaluator_Revive_Guard.Evaluator_Revive_Guard_C}
       }
      }]
    }
   ]
   decorators []
   properties {
    nodeName=On Ground
    children [
     {
      childComposite=null
      childTask=&self:BTTask_Wait_10
      decorators [
       &self:BTDecorator_Blackboard_12
       &self:BTDecorator_Loop_1
      ]
     }
     {
      childComposite=null
      childTask=&self:BTTask_Wait_0
      decorators [
       &self:BTDecorator_Blackboard_5
       &self:BTDecorator_Loop_2
      ]
     }
     {
      childComposite=null
      childTask=&self:FortAthenaBTTask_SteerMovement_0
      decorators [
       &self:BTDecorator_Blackboard_21
       &self:HasActiveAbilityWithAnyMatchingTag_C_0
      ]
     }
     {
      childComposite=null
      childTask=&self:FortAthenaBTTask_ShootTrap_0
      decorators [
       &self:BTDecorator_Blackboard_0
       &self:FortAthenaBTDecorator_BehaviorControls_2
      ]
     }
     {
      childComposite=null
      childTask=&self:FortAthenaBTTask_BotMoveTo_1
      decorators [
       &self:BTDecorator_Blackboard_3
       &self:FortAthenaBTDecorator_BehaviorControls_3
      ]
     }
     {
      childComposite=&self:BTComposite_Selector_6
      childTask=null
      decorators [
       &self:BTDecorator_Blackboard_2
       &self:BTDecorator_Loop_4
      ]
     }
     {
      childComposite=&self:BTComposite_Sequence_7
      childTask=null
      decorators [
       &self:BTDecorator_Blackboard_36
       &self:BTDecorator_Blackboard_37
      ]
     }
     {
      childComposite=&self:BTComposite_Selector_16
      childTask=null
      decorators [
       &self:BTDecorator_Blackboard_20
       &self:FortAthenaBTDecorator_BehaviorControls_17
       &self:BTDecorator_GameplayTagQuery_3
      ]
     }
     {
      childComposite=&self:BTComposite_Selector_4
      childTask=null
      decorators [
       &self:BTDecorator_Blackboard_4
       &self:FortAthenaBTDecorator_BehaviorControls_6
      ]
     }
     {childComposite=&self:BTComposite_Selector_15, childTask=null, decorators []}
    ]
    services [
     &self:FortAthenaBTService_AIEvaluator_19
     &self:FortAthenaBTService_UpdateTarget_1
     &self:FortAthenaBTService_AIEvaluator_28
     &self:FortAthenaBTService_AIEvaluator_34
     &self:FortAthenaBTService_AIEvaluator_39
     &self:FortAthenaBTService_AIEvaluator_227
     &self:FortAthenaBTService_Crouch_1
     &self:FortAthenaBTService_Slide_0
     &self:FortAthenaBTService_Sprinting_0
     &self:FortAthenaBTService_Jump_0
     &self:FortAthenaBTService_AIEvaluator_49
     &self:FortAthenaBTService_AIEvaluator_56
     &self:FortAthenaBTService_AIEvaluator_58
    ]
   }
   services [
    Evaluating... Leash {
     class=FortAthenaBTService_AIEvaluator
     properties {
      interval=0.5
      nodeName=Evaluating... Leash
      aIEvaluatorClass {defaultValue=/Game/Athena/AI/NPCs/Base/Evaluators/Evaluator_Leash_Guard.Evaluator_Leash_Guard_C}
     }
    }
    Evaluating...Targets {class=FortAthenaBTService_UpdateTarget, properties {interval=0.3}}
    Evaluating...AvoidThreat {
     class=FortAthenaBTService_AIEvaluator
     properties {
      interval=0.5
      nodeName=Evaluating...AvoidThreat
      aIEvaluatorClass {defaultValue=/Game/Athena/AI/Phoebe/Evaluators/Phoebe_Evaluator_AvoidThreat.Phoebe_Evaluator_AvoidThreat_C}
     }
    }
    Evaluating.. Propagate Awareness {
     class=FortAthenaBTService_AIEvaluator
     properties {
      interval=1
      nodeName=Evaluating.. Propagate Awareness
      aIEvaluatorClass {defaultValue=/Game/Athena/AI/NPCs/Base/Evaluators/Evaluator_PropagateAwareness_Guard.Evaluator_PropagateAwareness_Guard_C}
     }
    }
    Evaluating...Character Launched {
     class=FortAthenaBTService_AIEvaluator
     properties {
      interval=0.5
      nodeName=Evaluating...Character Launched
      aIEvaluatorClass {defaultValue=/Script/FortniteAIServer.FortAthenaAIBotEvaluator_CharacterLaunched}
     }
    }
    Evaluating... SelectVehicle {
     class=FortAthenaBTService_AIEvaluator
     properties {
      interval=0.5
      nodeName=Evaluating... SelectVehicle
      aIEvaluatorClass {defaultValue=/Game/Athena/AI/NPCs/Base/Evaluators/Evaluator_SelectConverterVehicle_NPC.Evaluator_SelectConverterVehicle_NPC_C}
     }
    }
    Executing...Crouch {class=FortAthenaBTService_Crouch}
    Executing...Slide {
     class=FortAthenaBTService_Slide
     properties {
      bCallTickOnSearchStart=false
      bRestartTimerOnEachActivation=false
      interval=0.5
      nodeName=Executing...Slide
      randomDeviation=0.1
      slideExecutionStatusName=AIEvaluator_Slide_ExecutionStatus
     }
    }
    Executing...Sprinting {class=FortAthenaBTService_Sprinting}
    Executing...Jump {class=FortAthenaBTService_Jump, properties {crouchExecutionStatusName=AIEvaluator_Crouch_ExecutionStatus}}
    Evaluating...Trap On Path {
     class=FortAthenaBTService_AIEvaluator
     properties {
      interval=0.3
      nodeName=Evaluating...Trap On Path
      aIEvaluatorClass {defaultValue=/Script/FortniteAIServer.FortAthenaAIBotEvaluator_TrapOnPathDetected}
     }
    }
    Evaluate...Sand Tunnel {
     class=FortAthenaBTService_AIEvaluator
     properties {
      interval=100000000
      nodeName=Evaluate...Sand Tunnel
      aIEvaluatorClass {defaultValue=/Game/Athena/AI/Phoebe/Evaluators/Phoebe_Evaluator_SandTunnel.Phoebe_Evaluator_SandTunnel_C}
     }
    }
    Evaluating...Ground {
     class=FortAthenaBTService_AIEvaluator
     properties {interval=1, nodeName=Evaluating...Ground, aIEvaluatorClass {defaultValue=/Script/FortniteAIServer.FortAthenaAIBotEvaluator_Ground}}
    }
   ]
  }
 ]
 decorators []
 properties {
  nodeName=Alive
  children [
   {
    childComposite=null
    childTask=&self:FortAthenaBTTask_Dive_0
    decorators [
     &self:BTDecorator_Blackboard_15
     &self:FortAthenaBTDecorator_BehaviorControls_0
    ]
   }
   {
    childComposite=null
    childTask=&self:FortAthenaBTTask_Glide_0
    decorators [
     &self:BTDecorator_Blackboard_22
     &self:FortAthenaBTDecorator_BehaviorControls_1
    ]
   }
   {
    childComposite=null
    childTask=&self:BTTask_RunBehavior_9
    decorators [
     &self:BTDecorator_Blackboard_42
     &self:BTDecorator_Loop_8
    ]
   }
   {
    childComposite=null
    childTask=&self:FortBTTask_InjectionPoint_3
    decorators [&self:FortBTDecorator_CanDoInjectedBehavior_5]
   }
   {childComposite=&self:BTComposite_Selector_7, childTask=null, decorators []}
  ]
  services [
   &self:FortAthenaBTService_AIEvaluator_1
   &self:FortBTService_AIEvaluatorsInjector_4
  ]
 }
 services [
  Evaluating...Free Fall {
   class=FortAthenaBTService_AIEvaluator
   properties {
    bCallTickOnSearchStart=true
    interval=0.5
    nodeName=Evaluating...Free Fall
    aIEvaluatorClass {defaultValue=/Game/Athena/AI/NPCs/Base/Evaluators/Evaluator_FreeFall_NPC.Evaluator_FreeFall_NPC_C}
   }
  }
  Behavior Injector {
   class=FortBTService_AIEvaluatorsInjector
   properties {
    bCallTickOnSearchStart=false
    bRestartTimerOnEachActivation=false
    interval=0.5
    nodeName=Behavior Injector
    randomDeviation=0.1
    injectionTagsKeys [{
      actionInterfaceObjectKeyName=Injection_Alive_HighPriority_ActionObject
      injectionKeyName=AIEvaluator_Injection_Alive_HighPriority
      injectionTag {tagName=Behavior.Injection.Alive.HighPriority}
     }]
   }
  }
 ]
}
