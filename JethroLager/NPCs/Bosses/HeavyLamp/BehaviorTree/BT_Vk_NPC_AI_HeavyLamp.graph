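# Behavior tree for the HeavyLamp boss NPC, built on Verse NPC evaluators and the BB_Vk_HeavyLamp blackboard.
# Root services update targets, evaluate combat, conversation and focus, execute weapon and movement actions, and provide injection slots.
# Their blackboard status keys gate a priority selector: conversation, patrol, attack movement, navigation with unstuck and zipline subtrees, emote/animation holds, revive, look-at, idle.
# That selector loops beneath the root so evaluators do not restart.
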BehaviorTree /JethroLager/NPCs/Bosses/HeavyLamp/BehaviorTree/BT_Vk_NPC_AI_HeavyLamp
blackboard=&/BB_Vk_HeavyLamp
root {
 name=Selector
 class=BTComposite_Selector
 children [Loop to prevent evaluator to restart {
   class=BTComposite_Selector
   children [
    Conversation {
     class=FortAthenaBTTask_Conversation
     children []
     decorators [
      Conversation? {
       class=BTDecorator_Blackboard
       properties {
        arithmeticOperation=GreaterOrEqual
        intValue=3
        nodeName=Conversation?
        stringValue=EExecutionStatus::ExecutionPending
        blackboardKey {selectedKeyName=AIEvaluator_Conversation_ExecutionStatus}
       }
      }
      "Behavior Controls = CanConverse" {
       class=FortAthenaBTDecorator_BehaviorControls
       properties {bInverseCondition=false, behaviorTreeBranch=CanConverse, flowAbortMode=Both, nodeName="Behavior Controls = CanConverse"}
      }
     ]
     properties {
      bIgnoreRestartSelf=false
      bResetEvaluatorStatusKeyOnFinish=true
      conversationStatusKeyName=AIEvaluator_Conversation_ExecutionStatus
      nodeName=Conversation
     }
     services []
    }
    Selector {
     class=BTComposite_Selector
     children [Navigate to Position {
       class=FortAthenaBTTask_MoveTo
       children []
       decorators []
       properties {
        bAllowPartialPathToUncompleteNavigationArea=false
        bAllowRestartOnPartialSuccess=false
        bIgnoreRestartSelf=false
        executionStatusKeyName=AIEvaluator_Patrol_ExecutionStatus
        movementResultKeyName=AIEvaluator_Patrol_MovementState
        nodeName=Navigate to Position
        acceptableRadius {defaultValue=5, key=Vk_NavigateToRadius}
        bAllowPartialPath {defaultValue=true, key=Vk_NavigateToAllowPartialPath}
        bAllowStrafe {defaultValue=false, key=VK_NavigateToAllowStrafe}
        bProjectGoalLocation {defaultValue=true, key=None}
        bReachTestIncludesAgentRadius {defaultValue=true, key=None}
        bReachTestIncludesGoalRadius {defaultValue=true, key=None}
        bRequireNavigableEndLocation {defaultValue=true, key=None}
        bStartFromPreviousPath {defaultValue=false, key=None}
        bTrackMovingGoal {defaultValue=true, key=None}
        blackboardKey {bNoneIsAllowedValue=false, selectedKeyName=AIEvaluator_Patrol_Destination}
        filterClass {baseClass=&NavigationQueryFilter, defaultValue=null, key=None}
        observedBlackboardValueTolerance {defaultValue=4.75, key=None}
       }
       services []
      }]
     decorators [PatrolAround Status {
       class=BTDecorator_Blackboard
       properties {
        arithmeticOperation=GreaterOrEqual
        intValue=3
        nodeName=PatrolAround Status
        stringValue=EExecutionStatus::ExecutionPending
        blackboardKey {selectedKeyName=Vk_PatrolAroundExecutionStatus}
       }
      }]
     properties {
      children [{
        childComposite=null
        childTask=&self:FortAthenaBTTask_MoveTo_2
        decorators []
       }]
      services [&self:FortAthenaBTService_AIEvaluator_22]
     }
     services [Evaluating... Patrol Around {
       class=FortAthenaBTService_AIEvaluator
       properties {
        bCallTickOnSearchStart=true
        interval=10
        nodeName=Evaluating... Patrol Around
        randomDeviation=0.5
        aIEvaluatorClass {defaultValue=/VerseNPC/BehaviorTrees/BP_Vk_RoamAroundEvaluator.BP_Vk_RoamAroundEvaluator_C}
       }
      }]
    }
    Move To Attack {
     class=BTComposite_Selector
     children [
      Move To Range {
       class=BTTask_RunBehavior
       decorators [
        Blackboard Based Condition {
         class=BTDecorator_Blackboard
         properties {
          arithmeticOperation=GreaterOrEqual
          flowAbortMode=None
          intValue=3
          nodeName=Blackboard Based Condition
          stringValue=EExecutionStatus::ExecutionPending
          blackboardKey {selectedKeyName=AIEvaluator_RangeAttack_ExecutionStatus}
         }
        }
        Blackboard Based Condition {
         class=BTDecorator_Blackboard
         properties {
          arithmeticOperation=GreaterOrEqual
          flowAbortMode=None
          intValue=3
          nodeName=Blackboard Based Condition
          stringValue=EExecutionStatus::ExecutionPending
          blackboardKey {selectedKeyName=AIEvaluator_ThrowableAttack_ExecutionStatus}
         }
        }
       ]
       properties {behaviorAsset=&BT_Phoebe_MoveToRange, nodeName=Move To Range}
      }
      Move To Melee {
       class=FortAthenaBTTask_BotMoveTo
       decorators [
        Can Do MeleeAttack {
         class=BTDecorator_Blackboard
         properties {
          arithmeticOperation=GreaterOrEqual
          intValue=3
          nodeName=Can Do MeleeAttack
          stringValue=EExecutionStatus::ExecutionPending
          blackboardKey {selectedKeyName=AIEvaluator_MeleeAttack_ExecutionStatus}
         }
        }
        Loop {class=BTDecorator_Loop}
       ]
       properties {
        movementResultKeyName=AIEvaluator_MeleeAttack_MovementState
        nodeName=Move To Melee
        acceptableRadius {defaultValue=5}
        allowRandomWobble {defaultValue=false}
        bAllowPartialPath {defaultValue=false}
        bAllowStrafe {defaultValue=false}
        blackboardKey {selectedKeyName=AIEvaluator_MeleeAttack_Destination}
        observedBlackboardValueTolerance {defaultValue=4}
       }
      }
      Fallback {class=BTTask_Wait, properties {nodeName=Fallback, waitTime {defaultValue=1}}}
     ]
     decorators [MoveToAttack Status {
       class=BTDecorator_Blackboard
       properties {
        arithmeticOperation=GreaterOrEqual
        intValue=3
        nodeName=MoveToAttack Status
        stringValue=EExecutionStatus::ExecutionPending
        blackboardKey {selectedKeyName=Vk_MoveToAttackExecutionStatus}
       }
      }]
     properties {
      nodeName=Move To Attack
      children [
       {
        childComposite=null
        childTask=&self:BTTask_RunBehavior_8
        decoratorOps [{number=2, operation=Or}, {number=0}, {number=1}]
        decorators [
         &self:BTDecorator_Blackboard_1
         &self:BTDecorator_Blackboard_0
        ]
       }
       {
        childComposite=null
        childTask=&self:FortAthenaBTTask_BotMoveTo_8
        decorators [
         &self:BTDecorator_Blackboard_28
         &self:BTDecorator_Loop_3
        ]
       }
       {childComposite=null, childTask=&self:BTTask_Wait_1, decorators []}
      ]
      services []
     }
     services []
    }
    Navigate To {
     class=BTComposite_Selector
     children [
      LeaveVehicle {
       class=FortAthenaBTTask_LeaveVehicle
       children []
       decorators [IsInVehicle {
         class=BTDecorator_Blackboard
         properties {arithmeticOperation=Equal, intValue=0, nodeName=IsInVehicle, stringValue="", blackboardKey {selectedKeyName=AIEvaluator_Vehicle_Actor}}
        }]
       properties {bIgnoreRestartSelf=false, bWaitVehicleStop=false, nodeName=LeaveVehicle}
       services []
      }
      Navigate to Entity {
       class=FortAthenaBTTask_MoveTo
       children []
       decorators [Is Destination Actor {
         class=BTDecorator_IsBBEntryOfClass
         properties {
          bInverseCondition=false
          flowAbortMode=Both
          nodeName=Is Destination Actor
          blackboardKey {bNoneIsAllowedValue=false, selectedKeyName=Vk_NavigateToTarget}
          testClass {baseClass=&Object, defaultValue=&Actor, key=None}
         }
        }]
       properties {
        bAllowPartialPathToUncompleteNavigationArea=true
        bAllowRestartOnPartialSuccess=false
        bIgnoreRestartSelf=false
        executionStatusKeyName=VK_NavigateToExecutionStatus
        movementResultKeyName=VK_NavigateToResult
        nodeName=Navigate to Entity
        acceptableRadius {defaultValue=5, key=Vk_NavigateToRadius}
        bAllowPartialPath {defaultValue=true, key=Vk_NavigateToAllowPartialPath}
        bAllowStrafe {defaultValue=false, key=VK_NavigateToAllowStrafe}
        bProjectGoalLocation {defaultValue=true, key=None}
        bReachTestIncludesAgentRadius {defaultValue=true, key=None}
        bReachTestIncludesGoalRadius {defaultValue=true, key=None}
        bRequireNavigableEndLocation {defaultValue=true, key=None}
        bStartFromPreviousPath {defaultValue=false, key=None}
        bTrackMovingGoal {defaultValue=true, key=None}
        blackboardKey {bNoneIsAllowedValue=false, selectedKeyName=Vk_NavigateToTarget}
        filterClass {baseClass=&NavigationQueryFilter, defaultValue=null, key=None}
        observedBlackboardValueTolerance {defaultValue=4.75, key=None}
       }
       services []
      }
      Selector {
       class=BTComposite_Selector
       children [
        Run Unstuck Behavior {
         class=BTTask_RunBehavior
         decorators []
         properties {behaviorAsset=&BT_Phoebe_UnstuckNew, nodeName=Run Unstuck Behavior}
        }
        Run Zipline Behavior {
         class=BTTask_RunBehavior
         decorators [
          Blackboard Based Condition {
           class=BTDecorator_Blackboard
           properties {
            arithmeticOperation=GreaterOrEqual
            intValue=4
            nodeName=Blackboard Based Condition
            stringValue=EExecutionStatus::ExecutionAllowed
            blackboardKey {selectedKeyName=AIEvaluator_Zipline_MoveExecutionStatus}
           }
          }
          Blackboard Based Condition {
           class=BTDecorator_Blackboard
           properties {
            arithmeticOperation=GreaterOrEqual
            intValue=4
            nodeName=Blackboard Based Condition
            stringValue=EExecutionStatus::ExecutionAllowed
            blackboardKey {selectedKeyName=AIEvaluator_Zipline_UsageExecutionStatus}
           }
          }
         ]
         properties {behaviorAsset=&BT_Phoebe_Zipline, nodeName=Run Zipline Behavior}
        }
        Navigate to Position {
         class=FortAthenaBTTask_MoveTo
         children []
         decorators []
         properties {
          bAllowPartialPathToUncompleteNavigationArea=true
          bAllowRestartOnPartialSuccess=false
          bIgnoreRestartSelf=false
          executionStatusKeyName=VK_NavigateToExecutionStatus
          movementResultKeyName=VK_NavigateToResult
          nodeName=Navigate to Position
          acceptableRadius {defaultValue=5, key=Vk_NavigateToRadius}
          bAllowPartialPath {defaultValue=true, key=Vk_NavigateToAllowPartialPath}
          bAllowStrafe {defaultValue=false, key=VK_NavigateToAllowStrafe}
          bProjectGoalLocation {defaultValue=true, key=None}
          bReachTestIncludesAgentRadius {defaultValue=true, key=None}
          bReachTestIncludesGoalRadius {defaultValue=true, key=None}
          bRequireNavigableEndLocation {defaultValue=true, key=None}
          bStartFromPreviousPath {defaultValue=false, key=None}
          bTrackMovingGoal {defaultValue=true, key=None}
          blackboardKey {bNoneIsAllowedValue=false, selectedKeyName=Vk_NavigateToDestination}
          filterClass {baseClass=&NavigationQueryFilter, defaultValue=null, key=None}
          observedBlackboardValueTolerance {defaultValue=4.75, key=None}
         }
         services []
        }
       ]
       decorators []
       properties {
        children [
         {childComposite=null, childTask=&self:BTTask_RunBehavior_11, decorators []}
         {
          childComposite=null
          childTask=&self:BTTask_RunBehavior_23
          decoratorOps [{number=2, operation=Or}, {number=0}, {number=1}]
          decorators [
           &self:BTDecorator_Blackboard_8
           &self:BTDecorator_Blackboard_9
          ]
         }
         {
          childComposite=null
          childTask=&self:FortAthenaBTTask_MoveTo_1
          decorators []
         }
        ]
        services [&self:FortAthenaBTService_AIEvaluator_2]
       }
       services [Evaluate... Zipline {
         class=FortAthenaBTService_AIEvaluator
         properties {
          bCallTickOnSearchStart=true
          interval=0.5
          nodeName=Evaluate... Zipline
          aIEvaluatorClass {defaultValue=/VerseNPC/BehaviorTrees/BT_Vk_Zipline_Evaluator.BT_Vk_Zipline_Evaluator_C}
         }
        }]
      }
     ]
     decorators [NavigateTo Status {
       class=BTDecorator_Blackboard
       properties {
        arithmeticOperation=GreaterOrEqual
        intValue=3
        nodeName=NavigateTo Status
        stringValue=EExecutionStatus::ExecutionPending
        blackboardKey {selectedKeyName=VK_NavigateToExecutionStatus}
       }
      }]
     properties {
      nodeName=Navigate To
      children [
       {
        childComposite=null
        childTask=&self:FortAthenaBTTask_LeaveVehicle_0
        decorators [&self:BTDecorator_Blackboard_2]
       }
       {
        childComposite=null
        childTask=&self:FortAthenaBTTask_MoveTo_0
        decorators [&self:BTDecorator_IsBBEntryOfClass_1]
       }
       {childComposite=&self:BTComposite_Selector_7, childTask=null, decorators []}
      ]
      services []
     }
     services []
    }
    Wait {
     class=BTTask_Wait
     decorators [PlayEmote Status {
       class=BTDecorator_Blackboard
       properties {
        arithmeticOperation=GreaterOrEqual
        intValue=3
        nodeName=PlayEmote Status
        stringValue=EExecutionStatus::ExecutionPending
        blackboardKey {selectedKeyName=VK_PlayEmoteExecutionStatus}
       }
      }]
     properties {waitTime {defaultValue=10000000000}}
    }
    Wait {
     class=BTTask_Wait
     decorators [PlayAnimation Status {
       class=BTDecorator_Blackboard
       properties {
        arithmeticOperation=GreaterOrEqual
        intValue=3
        nodeName=PlayAnimation Status
        stringValue=ExecutionPending
        blackboardKey {selectedKeyName=Vk_PlayAnimationExecutionStatus}
       }
      }]
     properties {waitTime {defaultValue=10000000000}}
    }
    Revive Interaction {
     class=FortAthenaBTTask_Interact
     children []
     decorators [Revive Status {
       class=BTDecorator_Blackboard
       properties {
        arithmeticOperation=GreaterOrEqual
        intValue=3
        nodeName=Revive Status
        stringValue=ExecutionPending
        blackboardKey {selectedKeyName=Vk_ReviveExecutionStatus}
       }
      }]
     properties {
      attemptInterval=0.25
      bIgnoreRestartSelf=false
      bShouldFocusOnInteraction=true
      focalPointName=None
      interactActorName=AIEvaluator_Global_InteractActor
      jumpExecutionStatusName=AIEvaluator_Jump_ExecutionStatus
      maxInteractAttempts=5
      nodeName=Revive Interaction
      weaponTriggerMeleeName=AIEvaluator_WeaponTriggerMelee_ExecutionStatus
      interactContextInfoKeySelector {bNoneIsAllowedValue=false, selectedKeyName=Vk_ReviveContext}
      interactExecutionStatusKeySelector {bNoneIsAllowedValue=false, selectedKeyName=Vk_ReviveExecutionStatus}
     }
     services []
    }
    Wait {
     class=BTTask_Wait
     decorators [Look At Status {
       class=BTDecorator_Blackboard
       properties {
        arithmeticOperation=GreaterOrEqual
        intValue=3
        nodeName=Look At Status
        notifyObserver=ValueChange
        stringValue=ExecutionPending
        blackboardKey {selectedKeyName=Vk_LookAtExecutionStatus}
       }
      }]
     properties {waitTime {defaultValue=10000000000}}
    }
    Wait {
     class=BTTask_Wait
     decorators [Idle Status {
       class=BTDecorator_Blackboard
       properties {
        arithmeticOperation=GreaterOrEqual
        intValue=3
        nodeName=Idle Status
        notifyObserver=ValueChange
        stringValue=ExecutionPending
        blackboardKey {selectedKeyName=Vk_IdleExecutionStatus}
       }
      }]
     properties {waitTime {defaultValue=10000000000}}
    }
    Idle {class=BTTask_Wait, properties {nodeName=Idle, waitTime {defaultValue=1}}}
   ]
   decorators [Loop {class=BTDecorator_Loop}]
   properties {
    nodeName=Loop to prevent evaluator to restart
    children [
     {
      childComposite=null
      childTask=&self:FortAthenaBTTask_Conversation_1
      decorators [
       &self:BTDecorator_Blackboard_33
       &self:FortAthenaBTDecorator_BehaviorControls_2
      ]
     }
     {
      childComposite=&self:BTComposite_Selector_5
      childTask=null
      decorators [&self:BTDecorator_Blackboard_31]
     }
     {
      childComposite=&self:BTComposite_Selector_2
      childTask=null
      decorators [&self:BTDecorator_Blackboard_23]
     }
     {
      childComposite=&self:BTComposite_Selector_6
      childTask=null
      decorators [&self:BTDecorator_Blackboard_12]
     }
     {
      childComposite=null
      childTask=&self:BTTask_Wait_4
      decorators [&self:BTDecorator_Blackboard_4]
     }
     {
      childComposite=null
      childTask=&self:BTTask_Wait_5
      decorators [&self:BTDecorator_Blackboard_6]
     }
     {
      childComposite=null
      childTask=&self:FortAthenaBTTask_Interact_1
      decorators [&self:BTDecorator_Blackboard_32]
     }
     {
      childComposite=null
      childTask=&self:BTTask_Wait_3
      decorators [&self:BTDecorator_Blackboard_5]
     }
     {
      childComposite=null
      childTask=&self:BTTask_Wait_2
      decorators [&self:BTDecorator_Blackboard_3]
     }
     {childComposite=null, childTask=&self:BTTask_Wait_0, decorators []}
    ]
    services []
   }
   services []
  }]
 decorators []
 properties {
  children [{
    childComposite=&self:BTComposite_Selector_3
    childTask=null
    decorators [&self:BTDecorator_Loop_4]
   }]
  services [
   &self:FortAthenaBTService_UpdateTarget_0
   &self:FortAthenaBTService_AIEvaluator_12
   &self:FortAthenaBTService_AIEvaluator_13
   &self:FortAthenaBTService_AIEvaluator_20
   &self:FortAthenaBTService_AIEvaluator_14
   &self:FortAthenaBTService_AIEvaluator_15
   &self:FortAthenaBTService_AIEvaluator_21
   &self:FortAthenaBTService_AIEvaluator_19
   &self:FortAthenaBTService_AIEvaluator_18
   &self:FortAthenaBTService_AIEvaluator_29
   &self:FortAthenaBTService_Crouch_2
   &self:FortAthenaBTService_ManageWeapon_1
   &self:FortAthenaBTService_Jump_1
   &self:FortAthenaBTService_Slide_1
   &self:FortAthenaBTService_Sprinting_1
   &self:FortAthenaBTService_Revive_1
   &self:FortAthenaBTService_Clamber_1
   &self:FortAthenaBTService_AIEvaluator_17
   &self:FortAthenaBTService_AIEvaluator_34
  ]
 }
 services [
  Evaluating...Targets {class=FortAthenaBTService_UpdateTarget, properties {interval=0.3}}
  Evaluating...Melee Attack {
   class=FortAthenaBTService_AIEvaluator
   properties {
    interval=0.3
    nodeName=Evaluating...Melee Attack
    aIEvaluatorClass {defaultValue=/VerseNPC/BehaviorTrees/BP_Vk_MeleeAttackEvaluator.BP_Vk_MeleeAttackEvaluator_C}
   }
  }
  Evaluating...Range Attack {
   class=FortAthenaBTService_AIEvaluator
   properties {
    interval=0.3
    nodeName=Evaluating...Range Attack
    aIEvaluatorClass {defaultValue=/VerseNPC/BehaviorTrees/BP_Vk_RangeAttackEvaluator.BP_Vk_RangeAttackEvaluator_C}
   }
  }
  Evaluating...Select Weapon {
   class=FortAthenaBTService_AIEvaluator
   properties {
    interval=0.2
    nodeName=Evaluating...Select Weapon
    randomDeviation=0.05
    aIEvaluatorClass {defaultValue=/VerseNPC/BehaviorTrees/BP_Vk_WeaponSelectionEvaluator.BP_Vk_WeaponSelectionEvaluator_C}
   }
  }
  Evaluating...Throwable Attack {
   class=FortAthenaBTService_AIEvaluator
   properties {
    interval=0.2
    nodeName=Evaluating...Throwable Attack
    randomDeviation=0.05
    aIEvaluatorClass {defaultValue=/VerseNPC/BehaviorTrees/BP_Vk_ThrowableAttackEvaluator.BP_Vk_ThrowableAttackEvaluator_C}
   }
  }
  Evaluating...Sprint {
   class=FortAthenaBTService_AIEvaluator
   properties {
    interval=0.5
    nodeName=Evaluating...Sprint
    aIEvaluatorClass {defaultValue=/VerseNPC/BehaviorTrees/BT_Vk_SprintEvaluator.BT_Vk_SprintEvaluator_C}
   }
  }
  Evaluating...Conversation {
   class=FortAthenaBTService_AIEvaluator
   properties {
    interval=1
    nodeName=Evaluating...Conversation
    randomDeviation=0.2
    aIEvaluatorClass {defaultValue=/Script/FortniteAIServer.FortAthenaAIBotEvaluator_Conversation}
   }
  }
  "Evaluating... Slot: ChromeBlob" {
   class=FortAthenaBTService_AIEvaluator
   properties {
    interval=100000000
    nodeName=Evaluating... Slot: ChromeBlob
    aIEvaluatorClass {defaultValue=null}
    injectionTag {tagName=Athena.AI.BTInjection.Phoebe.Blob}
   }
  }
  "Evaluating... Slot: Traversal" {
   class=FortAthenaBTService_AIEvaluator
   properties {
    interval=100000000
    nodeName=Evaluating... Slot: Traversal
    aIEvaluatorClass {defaultValue=null}
    injectionTag {tagName=Athena.AI.BTInjection.Phoebe.Traversal}
   }
  }
  "Evaluating... Slot: Grind" {
   class=FortAthenaBTService_AIEvaluator
   properties {
    interval=100000000
    nodeName=Evaluating... Slot: Grind
    aIEvaluatorClass {defaultValue=null}
    injectionTag {tagName=Athena.AI.BTInjection.Phoebe.Grind}
   }
  }
  Executing...Crouch {class=FortAthenaBTService_Crouch, properties {crouchExecutionStatusName=VK_CrouchExecutionStatus}}
  Executing...Weapons Actions {class=FortAthenaBTService_ManageWeapon, properties {interval=1, randomDeviation=0.3}}
  Executing...Jump {class=FortAthenaBTService_Jump, properties {crouchExecutionStatusName=VK_CrouchExecutionStatus}}
  Executing...Slide {
   class=FortAthenaBTService_Slide
   properties {
    bCallTickOnSearchStart=false
    bRestartTimerOnEachActivation=false
    interval=0.5
    nodeName=Executing...Slide
    randomDeviation=0.1
    slideExecutionStatusName=VK_SlideExecutionStatus
   }
  }
  Executing...Sprinting {class=FortAthenaBTService_Sprinting}
  Executing...Revive {
   class=FortAthenaBTService_Revive
   properties {
    bCallTickOnSearchStart=false
    bDisableLeash=false
    bRestartTimerOnEachActivation=false
    executionStatusName=Vk_ReviveExecutionStatus
    interactionContextInfoName=Vk_ReviveContext
    interactionExecutionStatusName=Vk_ReviveExecutionStatus
    interval=0.5
    moveToPathMovementStateName=None
    nodeName=Executing...Revive
    randomDeviation=0.1
    reviveTargetKeyName=Vk_ReviveTarget
   }
  }
  Executing... Clamber {
   class=FortAthenaBTService_Clamber
   properties {
    bCallTickOnSearchStart=false
    bRestartTimerOnEachActivation=false
    clamberAbilityStatusName=AIEvaluator_Clamber_AbilityStatus
    clamberAttemptRetryMaxCount=2
    clamberDestinationLocationName=AIEvaluator_Clamber_Destination
    clamberExecutionStatusName=AIEvaluator_Clamber_ExecutionStatus
    clamberOriginLocationName=AIEvaluator_Clamber_Origin
    clamberRetryDelay=0.15
    crouchExecutionStatusName=VK_CrouchExecutionStatus
    firstJumpClamberMaxStartDelay=2
    firstJumpRetryDelay=0.1
    firstJumpRetryMaxCount=5
    interval=0.5
    jumpExecutionStatusName=AIEvaluator_Jump_ExecutionStatus
    nodeName=Executing... Clamber
    randomDeviation=0.1
   }
  }
  Executing... Focus {
   class=FortAthenaBTService_AIEvaluator
   properties {interval=0.5, nodeName=Executing... Focus, aIEvaluatorClass {defaultValue=/VerseNPC/BehaviorTrees/BP_Vk_FocusOnTarget.BP_Vk_FocusOnTarget_C}}
  }
  Evaluating.. Propagate Awareness {
   class=FortAthenaBTService_AIEvaluator
   properties {
    interval=1
    nodeName=Evaluating.. Propagate Awareness
    aIEvaluatorClass {defaultValue=/Game/Athena/AI/NPCs/Base/Evaluators/Evaluator_PropagateAwareness_Guard.Evaluator_PropagateAwareness_Guard_C}
   }
  }
 ]
}
