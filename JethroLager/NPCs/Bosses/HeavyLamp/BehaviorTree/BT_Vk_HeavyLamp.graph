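# Top-level behavior tree for the HeavyLamp boss NPC, choosing between its default behavior and a Vk-specific AI behavior.
# A root selector runs the BT_HeavyLamp subtree when the VK_DefaultBehaviorExecutionStatus blackboard value is at or above ExecutionPending.
# If that check or subtree fails, the selector runs the BT_Vk_NPC_AI_HeavyLamp subtree, wrapping the default tree without duplicating its logic.
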BehaviorTree /JethroLager/NPCs/Bosses/HeavyLamp/BehaviorTree/BT_Vk_HeavyLamp
blackboard=&/BB_Vk_HeavyLamp
root {
 name=Selector
 class=BTComposite_Selector
 children [
  Run Default behavior {
   class=BTTask_RunBehavior
   decorators [Default Behavior Status {
     class=BTDecorator_Blackboard
     properties {
      arithmeticOperation=GreaterOrEqual
      intValue=3
      nodeName=Default Behavior Status
      stringValue=EExecutionStatus::ExecutionPending
      blackboardKey {selectedKeyName=VK_DefaultBehaviorExecutionStatus}
     }
    }]
   properties {behaviorAsset=&/BT_HeavyLamp, nodeName=Run Default behavior}
  }
  Run Behavior {
   class=BTTask_RunBehavior
   decorators []
   properties {behaviorAsset=&/BT_Vk_NPC_AI_HeavyLamp, nodeName=Run Behavior}
  }
 ]
 decorators []
 properties {
  children [
   {
    childComposite=null
    childTask=&self:BTTask_RunBehavior_3
    decorators [&self:BTDecorator_Blackboard_47]
   }
   {childComposite=null, childTask=&self:BTTask_RunBehavior_9, decorators []}
  ]
  services []
 }
 services []
}
